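# Material function for the Morning Soak Cook body cosmetic that produces blinking red light emissive, with parameters named for backpack lights, plus a LightMask output.
# A channel-selected mask texture is sampled with and without BumpOffset, then multiplied to fake recessed light depth.
# Time-driven sine math forms the blink pulse, while fresnel terms shape edges and brightness.
# Named reroutes link mask, pulse and min/max color blends into MF_FX_EmissiveAdjustments for the final emissive.
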MaterialFunction /BRCosmetics/Characters/Player/Male/Large/Bodies/M_LRG_MorningSoak_Cook/Materials/MF__MorningSoak_lightBlinkingFunction
graph {nodes [
  n1 = Multiply(A=NamedRerouteUsage(), B=NamedRerouteUsage())
  n2 = NamedRerouteUsage()
  FunctionOutput(None=MaterialFunctionCall(Emissive=LinearInterpolate(B=@n1, Alpha=@n2)) {materialFunction=&MF_FX_EmissiveAdjustments}.Result)
  n3 = TextureSampleParameter2D(UVs=BumpOffset(
    Height=ScalarParameter() {defaultValue=-0.1129, group=flickering lights, parameterName=Blink_LightDepth}
    HeightRatioInput=ScalarParameter() {defaultValue=0.05, group=flickering lights, parameterName=Blink_LightHeightRation}
   )) {
   group=flickering lights
   parameterName=Light Mask
   samplerType=SAMPLERTYPE_Masks
   texture=/BRCosmetics/Characters/Player/Male/Large/Bodies/M_LRG_MorningSoak_Cook/Textures/T_MorningSoak_Cook_FaceAcc_E
  }
  n4 = TextureSampleParameter2D() {
   group=flickering lights
   parameterName=Light Mask
   samplerType=SAMPLERTYPE_Masks
   texture=&T_MorningSoak_Cook_FaceAcc_E
  }
  n5 = ChannelMaskParameter(None=@n4.RGB) {group=flickering lights, parameterName=ChannelSelect}
  n6 = Multiply(A=ChannelMaskParameter(None=@n3.RGB) {group=flickering lights, parameterName=ChannelSelect}.R, B=Saturate(None=Multiply(A=@n5.R) {constB=3}))
  n7 = Saturate(None=Multiply(A=@n6) {constB=3})
  FunctionOutput(None=@n7) {outputName=LightMask}
  n8 = Constant() {r=0.8}
  n9 = Constant() {r=1}
  n10 = NamedRerouteUsage()
  n11 = Constant() {r=1}
  n12 = Constant() {r=1.8}
  n13 = Power(Base=Multiply(A=MaterialFunctionCall(Brightness=@n11, Exponent=@n12) {materialFunction=&MF_Fresnel_FX}.Fresnel) {constB=10}) {constExponent=20}
  NamedRerouteDeclaration(None=LinearInterpolate(A=LinearInterpolate(A=@n8, B=@n9, Alpha=@n10) {constB=0.8}, Alpha=Saturate(None=@n13))) {name=Edge, nodeColor=<rgba 1 0 0.9609 1>}
  n14 = LinearInterpolate(Alpha=NamedRerouteUsage()) {constA=2, constB=4}
  NamedRerouteDeclaration(None=Saturate(None=Multiply(A=@n6, B=@n14))) {name=Emissive_Tex, nodeColor=<rgba 0.4531 0 1 1>}
  NamedRerouteDeclaration(None=MaterialFunctionCall!(
    Thickness=ScalarParameter() {defaultValue=1, group=flickering lights, parameterName=Blink_Thickness}
    NormalWS=VertexNormalWS()
    PowerExponent=ScalarParameter() {defaultValue=4, group=flickering lights, parameterName=Blink_PowerExponet}
   ) {materialFunction=/Game/Weapons/Prototype/ObjectMover/Materials/MF_NormalizedFresnel}.Power) {name=NormalizeFresnel, nodeColor=<rgba 1 0 0.8438 1>}
  n15 = Multiply(A=Time(), B=ScalarParameter() {defaultValue=2, group=flickering lights, parameterName=Blink_duration})
  n16 = MaterialFunctionCall(Value=@n15) {materialFunction=&LinearSine}
  n17 = Abs(None=Sine(None=@n15))
  n18 = Max(A=Subtract(A=Multiply(A=@n17) {constB=2}), B=Constant())
  n19 = ScalarParameter() {defaultValue=45, group=flickering lights, parameterName=Blink_Rate of change}
  n20 = Saturate(None=OneMinus(None=Power(Base=@n18, Exp=@n19) {constExponent=3}))
  NamedRerouteDeclaration(None=Multiply(A=Multiply(A=@n16.Direction, B=@n20)) {constB=0.4}) {name=Blinking, nodeColor=<rgba 1 0.1875 0 1>}
  n21 = NamedRerouteUsage()
  n22 = Multiply(A=@n6, B=LinearInterpolate(Alpha=@n21) {constA=1, constB=4}) {constB=2}
  n23 = LinearInterpolate(Alpha=@n21) {constA=1, constB=4}
  NamedRerouteDeclaration(None=Saturate(None=Power(Base=@n22, Exp=@n23) {constExponent=1})) {name=EmissiveMask, nodeColor=<rgba 0.3594 1 0 1>}
  n24 = VectorParameter() {group=flickering lights, parameterName=BackPack_Red_lightEmissiveColor, defaultValue=<rgba 0.975 0.0244 0.0244 1>}
  n25 = VectorParameter() {group=flickering lights, parameterName=BackPack_Red_lightEmissiveColor2, defaultValue=<rgba 0.975 0.0244 0.0244 1>}
  n26 = LinearInterpolate(
   A=Multiply(A=@n24.RGB, B=ScalarParameter() {defaultValue=0.8, group=flickering lights, parameterName=Blink_Light_min})
   B=Multiply(A=@n25.RGB, B=ScalarParameter() {defaultValue=8, group=flickering lights, parameterName=Blink_Light_max})
   Alpha=NamedRerouteUsage()
  )
  n27 = NamedRerouteUsage()
  n28 = NamedRerouteUsage()
  n29 = Constant3Vector() {constant=<rgba 0.35 0 0.09 1>}
  n30 = NamedRerouteUsage()
  n31 = LinearInterpolate(A=Multiply(A=@n26, B=Multiply(A=@n27, B=@n26)), B=Add(A=Multiply(A=@n28, B=@n29), B=Multiply(A=@n28, B=@n26)), Alpha=@n30) {constA=1}
  n32 = LinearInterpolate(A=Multiply(A=@n31) {constB=1.2}, B=Multiply(A=@n31) {constB=1.5}, Alpha=@n30)
  n33 = LinearInterpolate(
   A=ScalarParameter() {defaultValue=0.2, group=flickering lights, parameterName=BackPack_FresnelLow}
   Alpha=MaterialFunctionCall(
    Brightness=ScalarParameter() {defaultValue=1.1, group=flickering lights, parameterName=BackPack_FresnelBrightness}
    Exponent=ScalarParameter() {defaultValue=4, group=flickering lights, parameterName=BackPack_FresnelExponent}
   ) {materialFunction=&MF_Fresnel_FX}.Fresnel
  )
  NamedRerouteDeclaration(None=Multiply(A=NamedRerouteUsage(), B=Multiply(A=@n32, B=@n33))) {name=EmissiveColor_output, nodeColor=<rgba 0 1 0.0391 1>}
  NamedRerouteDeclaration(None=@n7) {name=Mask, nodeColor=<rgba 1 0 0.8437 1>}
 ]}
