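# Physics asset for the right arm mesh of the MorningSoak Cook large male cosmetic body, covering a dangling shoulder cord.
# Capsule bodies form a two-link cord chain from a tiny master body, dyn_shoulder_cord_MSTR, through cord bones 01 and 02.
# Both joints set tight twist and first-swing limits with a wider second swing, favouring motion in one plane.
# A large upper-arm capsule has no constraint, apparently serving as a collider for the cord.
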PhysicsAsset /BRCosmetics/Characters/Player/Male/Large/Bodies/M_LRG_MorningSoak_Cook/Meshes/M_LRG_MorningSoak_Cook_Arm_R_Physics
bodies {
 dyn_shoulder_cord_01 [{shapeType=Capsule, length=17.146, radius=6.1071, shapeName=dyn_shoulder_cord_01_capsule, center=<5.7066 0 3.75>, rotation=<rot 0 0 90>}]
 dyn_shoulder_cord_02 [{shapeType=Capsule, length=14.2693, radius=5.8463, shapeName=dyn_shoulder_cord_02_capsule, center=<4.7112 0 0.08>, rotation=<rot 0 0 90>}]
 dyn_shoulder_cord_MSTR [{shapeType=Capsule, length=0.505, radius=0.505, shapeName=dyn_shoulder_cord_MSTR_capsule, rotation=<rot 0 -90 0>}]
 upperarm_r [{shapeType=Capsule, length=32.6718, radius=9.5883, shapeName=upperarm_r_capsule, center=<-9.103 0 0>, rotation=<rot 90 0 0>}]
}
constraints [
 {bone1Name=dyn_shoulder_cord_01, bone2Name=dyn_shoulder_cord_MSTR, swing1LimitDegrees=1, swing2LimitDegrees=30, twistLimitDegrees=1}
 {bone1Name=dyn_shoulder_cord_02, bone2Name=dyn_shoulder_cord_01, swing1LimitDegrees=1, swing2LimitDegrees=30, twistLimitDegrees=1}
]
