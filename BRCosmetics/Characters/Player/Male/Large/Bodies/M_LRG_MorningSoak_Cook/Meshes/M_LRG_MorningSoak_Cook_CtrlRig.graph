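# Corrective Control Rig for the MorningSoak Cook large male body, deforming its box costume pieces and attached props as the character moves.
# It runs the shared DeformRig function and fixed box offsets, then spherical pose readers on limb bones weight box and shoulder-pad corrections, some interpolated.
# Custom-idle and crouch-move flags adjust shoulder, torso and pelvis corrections.
# Finally, weighted parent constraints make accessories follow the deformed boxes and body bones.
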ControlRigBlueprint /BRCosmetics/Characters/Player/Male/Large/Bodies/M_LRG_MorningSoak_Cook/Meshes/M_LRG_MorningSoak_Cook_CtrlRig
graphs [
 New Function_ContainedGraph {nodes [
   RigVMFunctionEntryNode()
   > FN_DeformRig_FNC = RigVMFunctionReferenceNode(
    DeformRig_ Curve="(deformRig_l_upperarm_up_alpha_72_#1=False,deformRig_r_upperarm_up_alpha_73_#2=False,deformRig_l_upperarm_fr_alpha_74_#3=False,deformRig_r_upperarm_fr_alpha_75_CB51B8 ...(truncated)"
    DeformRig_Bool="(upperarm_1_#4=True,elbow_3_#5=True,wrist_7_#6=True,groin_8_#7=True,glute_17_#8=True,knee_9_40C408324CA261 ...(truncated)"
    DeformRig_Debug=(upperarm_debug_18_#4=False,wrist_debug_20_#6=False,lat_debug_26_#9=False,groinGlute_debug_27_#7=False)
    DeformRig_Value="(upperarm_up_18_#10=(Rotation=(X=0,Y=0,Z=0,W=1),Translation=(X=0,Y=0,Z=8),Scale3D=(X=0.7,Y=1,Z=1)),upperarm_fwd_fr_19_#11=(Rotation=(X=0,Y=0,Z=0,W=1),Translation=(X=0,Y=10,Z=0),Scale3D=(X=0.5,Y=1,Z= ...(truncated)"
   )
   > &RigUnit_ModifyTransforms(
    ItemToModify=((Item=(Type=Bone,Name="deform_pelvis_box_r"),Transform=(Rotation=(X=0,Y=0,Z=0,W=1),Translation=(X=2,Y=0,Z=0),Scale3D=(X=1,Y=1,Z=1))))
    Weight=1
    WeightMinimum=0
   )
   > &RigUnit_ModifyTransforms(
    ItemToModify=((Item=(Type=Bone,Name="deform_pelvis_box_l"),Transform=(Rotation=(X=0,Y=0,Z=0,W=1),Translation=(X=2,Y=0,Z=0),Scale3D=(X=1,Y=1,Z=1))))
    Weight=1
    WeightMinimum=0
   )
   > &RigUnit_ModifyTransforms(
    ItemToModify=((Item=(Type=Bone,Name="deform_pelvis_box_front"),Transform=(Rotation=(X=0,Y=0,Z=0,W=1),Translation=(X=1.5,Y=0,Z=0),Scale3D=(X=1,Y=1,Z=1))))
    Weight=1
    WeightMinimum=0
   )
   > &RigUnit_ModifyTransforms(
    ItemToModify=((Item=(Type=Bone,Name="deform_pelvis_box_back"),Transform=(Rotation=(X=0,Y=0,Z=0,W=1),Translation=(X=1,Y=0,Z=0),Scale3D=(X=1,Y=1,Z=1))))
    Weight=1
    WeightMinimum=0
   )
   > &RigUnit_ModifyTransforms(
    ItemToModify="((Item=(Type=Bone,Name=\"deform_lowerarm_box_bottom_r\"),Transform=(Rotation=(X=0,Y=0,Z=0,W=1),Translation=(X=0,Y=0,Z=0),Scale3D=(X=1,Y=1.05,Z=1.05))),(Item=(Type=Bone,Name=\"deform_lowerarm_box_bottom_l\"),Transform=(Rotation=(X=0,Y=0,Z=0,W=1),Translation=(X= ...(truncated)"
    Weight=1
    WeightMinimum=0
   )
   > RigVMFunctionReturnNode()
  ]}
 New Function_ContainedGraph {nodes [
   n1 = RigVMFunctionEntryNode()
   > RigVMFunction_Sequence_1 = RigVMAggregateNode()
   n2 = &RigUnit_SphericalPoseReader(
    ActiveRegionScaleFactors=(PositiveWidth=3,NegativeWidth=1,PositiveHeight=1.2,NegativeHeight=1)
    ActiveRegionSize=0.1
    Debug=(bDrawDebug=false,bDraw2D=False,bDrawLocalAxes=False,DebugScale=25,DebugSegments=20,DebugThickness=0.25)
    DriverAxis=(X=-1,Y=0,Z=0)
    DriverItem=(Type=Bone,Name="thigh_l")
    FalloffRegionScaleFactors=(PositiveWidth=1,NegativeWidth=1,PositiveHeight=0.75,NegativeHeight=0.75)
    FalloffSize=0.5
    FlipHeightScaling=False
    FlipWidthScaling=False
    OptionalParentItem=(Type=Bone,Name="pelvis")
    RotationOffset=(X=90,Y=0,Z=-15)
    ExecutePin=@RigVMFunction_Sequence_1.A
   )
   > &RigUnit_ModifyTransforms(
    ItemToModify="((Item=(Type=Bone,Name=\"deform_pelvis_box_MSTR\"),Transform=(Rotation=(X=0,Y=0,Z=0,W=1),Translation=(X=2.5,Y=0,Z=0),Scale3D=(X=1,Y=1,Z=1))),(Item=(Type=Bone,Name=\"deform_pelvis_box_front\"),Transform=(Rotation=(X=0,Y=-0.0698,Z=0,W=0.9976),Translation=(X=0,Y= ...(truncated)"
    WeightMinimum=0
    Weight=@n2.OutputParam
   )
   > n3 = &RigUnit_SphericalPoseReader(
    ActiveRegionScaleFactors=(PositiveWidth=3,NegativeWidth=1,PositiveHeight=1.3,NegativeHeight=1.2)
    ActiveRegionSize=0.1
    Debug=(bDrawDebug=false,bDraw2D=False,bDrawLocalAxes=False,DebugScale=25,DebugSegments=20,DebugThickness=0.25)
    DriverAxis=(X=-1,Y=0,Z=0)
    DriverItem=(Type=Bone,Name="thigh_l")
    FalloffRegionScaleFactors=(PositiveWidth=1,NegativeWidth=1,PositiveHeight=0.5,NegativeHeight=0.5)
    FalloffSize=0.4
    FlipHeightScaling=False
    FlipWidthScaling=False
    OptionalParentItem=(Type=Bone,Name="pelvis")
    RotationOffset=(X=130,Y=20,Z=-20)
   )
   > &RigUnit_ModifyTransforms(
    ItemToModify="((Item=(Type=Bone,Name=\"deform_pelvis_box_MSTR\"),Transform=(Rotation=(X=0,Y=0,Z=0,W=1),Translation=(X=2.5,Y=0,Z=0),Scale3D=(X=1,Y=1,Z=1))),(Item=(Type=Bone,Name=\"deform_pelvis_box_l\"),Transform=(Rotation=(X=-0.0846,Y=0.1783,Z=0.0514,W=-0.979),Translation=( ...(truncated)"
    WeightMinimum=0
    Weight=@n3.OutputParam
   )
   > n4 = &RigUnit_SphericalPoseReader(
    ActiveRegionScaleFactors=(PositiveWidth=1,NegativeWidth=1,PositiveHeight=1,NegativeHeight=1)
    ActiveRegionSize=0.25
    Debug=(bDrawDebug=false,bDraw2D=False,bDrawLocalAxes=False,DebugScale=25,DebugSegments=20,DebugThickness=0.25)
    DriverAxis=(X=-1,Y=0,Z=0)
    DriverItem=(Type=Bone,Name="thigh_l")
    FalloffRegionScaleFactors=(PositiveWidth=1,NegativeWidth=1,PositiveHeight=1,NegativeHeight=1)
    FalloffSize=0.5
    FlipHeightScaling=False
    FlipWidthScaling=False
    OptionalParentItem=(Type=Bone,Name="pelvis")
    RotationOffset=(X=109.0855,Y=0,Z=45.5889)
   )
   > &RigUnit_ModifyTransforms(
    ItemToModify="((Item=(Type=Bone,Name=\"deform_pelvis_box_l\"),Transform=(Rotation=(X=0,Y=0,Z=0,W=1),Translation=(X=1,Y=0,Z=0),Scale3D=(X=1,Y=1,Z=1))),(Item=(Type=Bone,Name=\"deform_pelvis_box_front\"),Transform=(Rotation=(X=0,Y=0,Z=0,W=1),Translation=(X=1,Y=0,Z=0),Scale3D=( ...(truncated)"
    WeightMinimum=0
    Weight=@n4.OutputParam
   )
   > n5 = &RigUnit_SphericalPoseReader(
    ActiveRegionScaleFactors=(PositiveWidth=3,NegativeWidth=1,PositiveHeight=1,NegativeHeight=1)
    ActiveRegionSize=0.15
    Debug=(bDrawDebug=false,bDraw2D=False,bDrawLocalAxes=False,DebugScale=25,DebugSegments=20,DebugThickness=0.25)
    DriverAxis=(X=-1,Y=0,Z=0)
    DriverItem=(Type=Bone,Name="thigh_l")
    FalloffRegionScaleFactors=(PositiveWidth=1,NegativeWidth=1,PositiveHeight=1,NegativeHeight=1)
    FalloffSize=0.4
    FlipHeightScaling=False
    FlipWidthScaling=False
    OptionalParentItem=(Type=Bone,Name="pelvis")
    RotationOffset=(X=180,Y=0,Z=0)
   )
   > &RigUnit_ModifyTransforms(
    ItemToModify="((Item=(Type=Bone,Name=\"deform_pelvis_box_MSTR\"),Transform=(Rotation=(X=0,Y=0,Z=0,W=1),Translation=(X=2.5,Y=0,Z=0),Scale3D=(X=1,Y=1,Z=1))),(Item=(Type=Bone,Name=\"deform_pelvis_box_l\"),Transform=(Rotation=(X=-0.0508,Y=0.0886,Z=-0.0045,W=-0.9948),Translation ...(truncated)"
    WeightMinimum=0
    Weight=@n5.OutputParam
   )
   > n6 = &RigUnit_SphericalPoseReader(
    ActiveRegionScaleFactors=(PositiveWidth=3,NegativeWidth=1,PositiveHeight=1.2,NegativeHeight=1)
    ActiveRegionSize=0.1
    Debug=(bDrawDebug=false,bDraw2D=False,bDrawLocalAxes=False,DebugScale=25,DebugSegments=20,DebugThickness=0.25)
    DriverAxis=(X=-1,Y=0,Z=0)
    DriverItem=(Type=Bone,Name="thigh_l")
    FalloffRegionScaleFactors=(PositiveWidth=1,NegativeWidth=1,PositiveHeight=0.75,NegativeHeight=0.75)
    FalloffSize=0.5
    FlipHeightScaling=False
    FlipWidthScaling=False
    OptionalParentItem=(Type=Bone,Name="pelvis")
    RotationOffset=(X=60,Y=0,Z=-15)
   )
   > &RigUnit_ModifyTransforms(
    ItemToModify="((Item=(Type=Bone,Name=\"deform_pelvis_box_MSTR\"),Transform=(Rotation=(X=0,Y=0,Z=0,W=1),Translation=(X=2.5,Y=0,Z=0),Scale3D=(X=1,Y=1,Z=1))),(Item=(Type=Bone,Name=\"deform_pelvis_box_l\"),Transform=(Rotation=(X=0,Y=0,Z=0,W=1),Translation=(X=2,Y=0,Z=0),Scale3D= ...(truncated)"
    WeightMinimum=0
    Weight=@n6.OutputParam
   )
   > n7 = &RigUnit_SphericalPoseReader(
    ActiveRegionScaleFactors=(PositiveWidth=1,NegativeWidth=1,PositiveHeight=1,NegativeHeight=0.5)
    ActiveRegionSize=0.2
    Debug=(bDrawDebug=false,bDraw2D=False,bDrawLocalAxes=False,DebugScale=25,DebugSegments=20,DebugThickness=0.25)
    DriverAxis=(X=1,Y=0,Z=0)
    DriverItem=(Type=Bone,Name="upperarm_l")
    FalloffRegionScaleFactors=(PositiveWidth=1,NegativeWidth=1,PositiveHeight=1,NegativeHeight=1)
    FalloffSize=0.4
    FlipHeightScaling=False
    FlipWidthScaling=False
    OptionalParentItem=(Type=Bone,Name="None")
    RotationOffset=(X=70.2746,Y=0,Z=-26.7803)
   )
   n8 = &RigVMFunction_AlphaInterp(InterpSpeedDecreasing=20, InterpSpeedIncreasing=20, Scale=1, bClampResult=False, bInterpResult=True, Value=@n7.OutputParam)
   > n9 = /Script/ControlRig.RigUnit_SetCurveValue(Curve=upperarm_l_up_SPR, Value=@n8.Result) {is=struct RigVMUnitNode}
   RigVMFunction_ControlFlowBranch = &RigVMFunction_ControlFlowBranch(Condition=@n1.CustomIdleIsPlaying, ExecuteContext=@n9.ExecutePin)
   &RigUnit_ModifyTransforms(
    ItemToModify="((Item=(Type=Bone,Name=\"deform_torso_box_MSTR\"),Transform=(Rotation=(X=0,Y=0,Z=0,W=1),Translation=(X=2,Y=0,Z=0),Scale3D=(X=1,Y=1,Z=1))),(Item=(Type=Bone,Name=\"deform_torso_box_front\"),Transform=(Rotation=(X=0,Y=0,Z=-0.1045,W=0.9945),Translation=(X=0,Y=0,Z= ...(truncated)"
    WeightMinimum=0
    ExecutePin=@RigVMFunction_ControlFlowBranch.Completed
    Weight=RigVMVariableNode() {variable=FEAdjustAlphaLeft}.Value
   )
   Subtract = &RigVMFunction_MathDoubleSub(A=1, B=RigVMVariableNode() {variable=FEAdjustAlphaLeft}.Value)
   > &RigUnit_ModifyTransforms(
    ItemToModify=((Item=(Type=Bone,Name="deform_pelvis_box_MSTR"),Transform=(Rotation=(X=0,Y=0,Z=0,W=1),Translation=(X=-0.75,Y=0,Z=0),Scale3D=(X=1,Y=1,Z=1))))
    WeightMinimum=0
    Weight=@Subtract.Result
   )
   > n10 = &RigUnit_SphericalPoseReader(
    ActiveRegionScaleFactors=(PositiveWidth=1,NegativeWidth=1,PositiveHeight=1,NegativeHeight=1)
    ActiveRegionSize=0.1
    Debug=(bDrawDebug=false,bDraw2D=False,bDrawLocalAxes=False,DebugScale=25,DebugSegments=20,DebugThickness=0.25)
    DriverAxis=(X=1,Y=0,Z=0)
    DriverItem=(Type=Bone,Name="lowerarm_l")
    FalloffRegionScaleFactors=(PositiveWidth=1,NegativeWidth=1,PositiveHeight=1,NegativeHeight=1)
    FalloffSize=0.2
    FlipHeightScaling=False
    FlipWidthScaling=False
    OptionalParentItem=(Type=Bone,Name="upperarm_l")
    RotationOffset=(X=90,Y=0,Z=-30)
   )
   > &RigUnit_ModifyTransforms(
    ItemToModify=((Item=(Type=Bone,Name="deform_lowerarm_box_top_l"),Transform=(Rotation=(X=0,Y=0,Z=0,W=-1),Translation=(X=3,Y=0,Z=0),Scale3D=(X=1,Y=1,Z=1))))
    WeightMinimum=0
    Weight=@n10.OutputParam
   )
   > n11 = &RigUnit_SphericalPoseReader(
    ActiveRegionScaleFactors=(PositiveWidth=1,NegativeWidth=1,PositiveHeight=1,NegativeHeight=1)
    ActiveRegionSize=0.1
    Debug=(bDrawDebug=false,bDraw2D=False,bDrawLocalAxes=False,DebugScale=25,DebugSegments=20,DebugThickness=0.25)
    DriverAxis=(X=-1,Y=0,Z=0)
    DriverItem=(Type=Bone,Name="calf_l")
    FalloffRegionScaleFactors=(PositiveWidth=1,NegativeWidth=1,PositiveHeight=1,NegativeHeight=1)
    FalloffSize=0.4
    FlipHeightScaling=False
    FlipWidthScaling=False
    OptionalParentItem=(Type=Bone,Name="thigh_l")
    RotationOffset=(X=-90,Y=0,Z=50)
   )
   > &RigUnit_ModifyTransforms(
    ItemToModify="((Item=(Type=Bone,Name=\"deform_calf_box_top_l\"),Transform=(Rotation=(X=0,Y=0,Z=0,W=-1),Translation=(X=-4,Y=0,Z=0),Scale3D=(X=1,Y=1,Z=1))),(Item=(Type=Bone,Name=\"dyn_calf_box_flap_top_bk_l\"),Transform=(Rotation=(X=0,Y=0.5664,Z=0,W=0.8241),Translation=(X=0,Y ...(truncated)"
    WeightMinimum=0
    Weight=@n11.OutputParam
   )
   > n12 = &RigUnit_SphericalPoseReader(
    ActiveRegionScaleFactors=(PositiveWidth=1,NegativeWidth=1,PositiveHeight=1,NegativeHeight=1)
    ActiveRegionSize=0.1
    Debug=(bDrawDebug=false,bDraw2D=False,bDrawLocalAxes=False,DebugScale=25,DebugSegments=20,DebugThickness=0.25)
    DriverAxis=(X=0,Y=1,Z=0)
    DriverItem=(Type=Bone,Name="foot_l")
    FalloffRegionScaleFactors=(PositiveWidth=1,NegativeWidth=1,PositiveHeight=1,NegativeHeight=1)
    FalloffSize=0.27
    FlipHeightScaling=False
    FlipWidthScaling=False
    OptionalParentItem=(Type=Bone,Name="calf_l")
    RotationOffset=(X=90,Y=0,Z=-50)
   )
   RigVMFunction_AlphaInterp = &RigVMFunction_AlphaInterp(InterpSpeedDecreasing=20, InterpSpeedIncreasing=20, Scale=1, bClampResult=False, bInterpResult=True, Value=@n12.OutputParam)
   > &RigUnit_ModifyTransforms(
    ItemToModify="((Item=(Type=Bone,Name=\"deform_calf_box_bottom_l\"),Transform=(Rotation=(X=0,Y=0,Z=0,W=-1),Translation=(X=6,Y=0,Z=0),Scale3D=(X=1,Y=1,Z=1))),(Item=(Type=Bone,Name=\"dyn_calf_box_flap_bottom_fr_l_MSTR\"),Transform=(Rotation=(X=-0.1324,Y=-0.6002,Z=-0.1016,W=0.7 ...(truncated)"
    WeightMinimum=0
    Weight=@RigVMFunction_AlphaInterp.Result
   )
   > n13 = &RigUnit_SphericalPoseReader(
    ActiveRegionScaleFactors=(PositiveWidth=1,NegativeWidth=1,PositiveHeight=1,NegativeHeight=1)
    ActiveRegionSize=0.1
    Debug=(bDrawDebug=false,bDraw2D=False,bDrawLocalAxes=False,DebugScale=25,DebugSegments=20,DebugThickness=0.25)
    DriverAxis=(X=0,Y=1,Z=0)
    DriverItem=(Type=Bone,Name="foot_l")
    FalloffRegionScaleFactors=(PositiveWidth=1,NegativeWidth=1,PositiveHeight=1,NegativeHeight=1)
    FalloffSize=0.25
    FlipHeightScaling=False
    FlipWidthScaling=False
    OptionalParentItem=(Type=Bone,Name="calf_l")
    RotationOffset=(X=90,Y=0,Z=55)
   )
   > &RigUnit_ModifyTransforms(
    ItemToModify=((Item=(Type=Bone,Name="deform_calf_box_bottom_l"),Transform=(Rotation=(X=0,Y=0,Z=0,W=1),Translation=(X=0,Y=0,Z=0),Scale3D=(X=1,Y=1.1,Z=1.1))))
    Weight=1
    WeightMinimum=0
   )
   RigVMVariableNode(Value=@n8.Result, ExecuteContext=@RigVMFunction_ControlFlowBranch.False) {variable=FEAdjustAlphaLeft}
   > &RigUnit_ModifyTransforms(
    ItemToModify=((Item=(Type=Bone,Name="deform_shoulderpad_l_MSTR"),Transform=(Rotation=(X=0.2358,Y=0,Z=0,W=0.9718),Translation=(X=2.6929,Y=-3.564,Z=-0.6952),Scale3D=(X=1,Y=1,Z=1))))
    WeightMinimum=0
    Weight=RigVMVariableNode() {variable=FEAdjustAlphaLeft}.Value
   )
   > n14 = &RigUnit_SphericalPoseReader(
    ActiveRegionScaleFactors=(PositiveWidth=1,NegativeWidth=1,PositiveHeight=1,NegativeHeight=1)
    ActiveRegionSize=0.15
    Debug=(bDrawDebug=false,bDraw2D=False,bDrawLocalAxes=False,DebugScale=25,DebugSegments=20,DebugThickness=0.25)
    DriverAxis=(X=1,Y=0,Z=0)
    DriverItem=(Type=Bone,Name="upperarm_l")
    FalloffRegionScaleFactors=(PositiveWidth=1,NegativeWidth=1,PositiveHeight=1,NegativeHeight=1)
    FalloffSize=0.25
    FlipHeightScaling=False
    FlipWidthScaling=False
    OptionalParentItem=(Type=Bone,Name="None")
    RotationOffset=(X=132.7257,Y=0,Z=0)
   )
   > &RigUnit_ModifyTransforms(
    ItemToModify=((Item=(Type=Bone,Name="deform_shoulderpad_l_MSTR"),Transform=(Rotation=(X=0.1869,Y=0,Z=0,W=0.9824),Translation=(X=0,Y=-3,Z=0),Scale3D=(X=1,Y=1,Z=1))))
    WeightMinimum=0
    Weight=@n14.OutputParam
   )
   > n15 = &RigUnit_SphericalPoseReader(
    ActiveRegionScaleFactors=(PositiveWidth=1,NegativeWidth=1,PositiveHeight=1,NegativeHeight=0.5)
    ActiveRegionSize=0.2
    Debug=(bDrawDebug=false,bDraw2D=False,bDrawLocalAxes=False,DebugScale=25,DebugSegments=20,DebugThickness=0.25)
    DriverAxis=(X=1,Y=0,Z=0)
    DriverItem=(Type=Bone,Name="upperarm_l")
    FalloffRegionScaleFactors=(PositiveWidth=1,NegativeWidth=1,PositiveHeight=1,NegativeHeight=1)
    FalloffSize=0.4
    FlipHeightScaling=False
    FlipWidthScaling=False
    OptionalParentItem=(Type=Bone,Name="None")
    RotationOffset=(X=20.9109,Y=-20.1439,Z=-26.7803)
   )
   > &RigUnit_ModifyTransforms(
    ItemToModify=((Item=(Type=Bone,Name="deform_shoulderpad_l_MSTR"),Transform=(Rotation=(X=0.0836,Y=0,Z=0,W=0.9965),Translation=(X=1,Y=0,Z=0),Scale3D=(X=0.9,Y=1,Z=1))))
    WeightMinimum=0
    Weight=@n15.OutputParam
   )
   RigVMVariableNode(ExecuteContext=@RigVMFunction_ControlFlowBranch.True) {variable=FEAdjustAlphaLeft}
   > n16 = &RigUnit_SphericalPoseReader(
    ActiveRegionScaleFactors=(PositiveWidth=1,NegativeWidth=1,PositiveHeight=1,NegativeHeight=1)
    ActiveRegionSize=0.1
    Debug=(bDrawDebug=true,bDraw2D=False,bDrawLocalAxes=False,DebugScale=25,DebugSegments=20,DebugThickness=0.25)
    DriverAxis=(X=1,Y=0,Z=0)
    DriverItem=(Type=Bone,Name="upperarm_l")
    FalloffRegionScaleFactors=(PositiveWidth=1,NegativeWidth=1,PositiveHeight=1,NegativeHeight=1)
    FalloffSize=0.3
    FlipHeightScaling=False
    FlipWidthScaling=False
    OptionalParentItem=(Type=Bone,Name="None")
    RotationOffset=(X=90,Y=0,Z=-45)
   )
   > &RigUnit_ModifyTransforms(
    ItemToModify=((Item=(Type=Bone,Name="deform_shoulderpad_l"),Transform=(Rotation=(X=0.136,Y=0.0502,Z=-0.1677,W=0.9751),Translation=(X=3.2117,Y=-3.4891,Z=-0.6152),Scale3D=(X=1,Y=1,Z=1))))
    WeightMinimum=0
    Weight=@n16.OutputParam
   )
   n17 = &RigUnit_SphericalPoseReader(
    ActiveRegionScaleFactors=(PositiveWidth=3,NegativeWidth=1,PositiveHeight=1.5,NegativeHeight=1.5)
    ActiveRegionSize=0.12
    Debug=(bDrawDebug=false,bDraw2D=False,bDrawLocalAxes=False,DebugScale=25,DebugSegments=20,DebugThickness=0.25)
    DriverAxis=(X=1,Y=0,Z=0)
    DriverItem=(Type=Bone,Name="thigh_r")
    FalloffRegionScaleFactors=(PositiveWidth=1,NegativeWidth=1,PositiveHeight=0.75,NegativeHeight=0.75)
    FalloffSize=0.6
    FlipHeightScaling=true
    FlipWidthScaling=true
    OptionalParentItem=(Type=Bone,Name="None")
    RotationOffset=(X=-88.1382,Y=-0.0285,Z=-10)
    ExecutePin=@RigVMFunction_Sequence_1.B
   )
   > &RigUnit_ModifyTransforms(
    ItemToModify="((Item=(Type=Bone,Name=\"deform_pelvis_box_MSTR\"),Transform=(Rotation=(X=0,Y=0,Z=0,W=1),Translation=(X=1,Y=0,Z=0),Scale3D=(X=1,Y=1,Z=1))),(Item=(Type=Bone,Name=\"deform_pelvis_box_front\"),Transform=(Rotation=(X=0,Y=-0.2224,Z=0,W=0.9749),Translation=(X=2.5107 ...(truncated)"
    WeightMinimum=0
    Weight=@n17.OutputParam
   )
   > n18 = &RigUnit_SphericalPoseReader(
    ActiveRegionScaleFactors=(PositiveWidth=3,NegativeWidth=1.5,PositiveHeight=1.3,NegativeHeight=1.3)
    ActiveRegionSize=0.12
    Debug=(bDrawDebug=false,bDraw2D=False,bDrawLocalAxes=False,DebugScale=25,DebugSegments=20,DebugThickness=0.25)
    DriverAxis=(X=1,Y=0,Z=0)
    DriverItem=(Type=Bone,Name="thigh_r")
    FalloffRegionScaleFactors=(PositiveWidth=1,NegativeWidth=1,PositiveHeight=0.75,NegativeHeight=0.75)
    FalloffSize=0.5
    FlipHeightScaling=true
    FlipWidthScaling=true
    OptionalParentItem=(Type=Bone,Name="None")
    RotationOffset=(X=-50,Y=5,Z=-30)
   )
   n19 = &RigVMFunction_AlphaInterp(InterpSpeedDecreasing=10, InterpSpeedIncreasing=10, Scale=1, bClampResult=False, bInterpResult=True, Value=@n18.OutputParam)
   > &RigUnit_ModifyTransforms(
    ItemToModify="((Item=(Type=Bone,Name=\"deform_pelvis_box_MSTR\"),Transform=(Rotation=(X=0,Y=0,Z=0,W=1),Translation=(X=1,Y=0,Z=0),Scale3D=(X=1,Y=1,Z=1))),(Item=(Type=Bone,Name=\"deform_pelvis_box_r\"),Transform=(Rotation=(X=-0.1118,Y=-0.206,Z=-0.0916,W=0.9678),Translation=(X ...(truncated)"
    WeightMinimum=0
    Weight=@n19.Result
   )
   > n20 = &RigUnit_SphericalPoseReader(
    ActiveRegionScaleFactors=(PositiveWidth=3,NegativeWidth=1,PositiveHeight=1,NegativeHeight=1)
    ActiveRegionSize=0.2
    Debug=(bDrawDebug=false,bDraw2D=False,bDrawLocalAxes=False,DebugScale=25,DebugSegments=20,DebugThickness=0.25)
    DriverAxis=(X=1,Y=0,Z=0)
    DriverItem=(Type=Bone,Name="thigh_r")
    FalloffRegionScaleFactors=(PositiveWidth=1,NegativeWidth=1,PositiveHeight=1,NegativeHeight=1)
    FalloffSize=0.5
    FlipHeightScaling=true
    FlipWidthScaling=true
    OptionalParentItem=(Type=Bone,Name="None")
    RotationOffset=(X=0,Y=-30,Z=0)
   )
   n21 = &RigVMFunction_AlphaInterp(InterpSpeedDecreasing=10, InterpSpeedIncreasing=10, Scale=1, bClampResult=False, bInterpResult=True, Value=@n20.OutputParam)
   > &RigUnit_ModifyTransforms(
    ItemToModify="((Item=(Type=Bone,Name=\"deform_pelvis_box_MSTR\"),Transform=(Rotation=(X=0,Y=0,Z=0,W=1),Translation=(X=1,Y=0,Z=0),Scale3D=(X=1,Y=1,Z=1))),(Item=(Type=Bone,Name=\"deform_pelvis_box_front\"),Transform=(Rotation=(X=0,Y=0,Z=0,W=1),Translation=(X=0,Y=0,Z=0),Scale3 ...(truncated)"
    WeightMinimum=0
    Weight=@n21.Result
   )
   > n22 = &RigUnit_SphericalPoseReader(
    ActiveRegionScaleFactors=(PositiveWidth=3,NegativeWidth=1,PositiveHeight=1,NegativeHeight=1)
    ActiveRegionSize=0.12
    Debug=(bDrawDebug=false,bDraw2D=False,bDrawLocalAxes=False,DebugScale=25,DebugSegments=20,DebugThickness=0.25)
    DriverAxis=(X=1,Y=0,Z=0)
    DriverItem=(Type=Bone,Name="thigh_r")
    FalloffRegionScaleFactors=(PositiveWidth=1,NegativeWidth=1,PositiveHeight=0.75,NegativeHeight=0.75)
    FalloffSize=0.6
    FlipHeightScaling=true
    FlipWidthScaling=true
    OptionalParentItem=(Type=Bone,Name="None")
    RotationOffset=(X=-120,Y=0,Z=-20)
   )
   > &RigUnit_ModifyTransforms(
    ItemToModify=((Item=(Type=Bone,Name="deform_pelvis_box_MSTR"),Transform=(Rotation=(X=0,Y=0,Z=0,W=1),Translation=(X=2.5,Y=0,Z=0),Scale3D=(X=1,Y=1,Z=1))))
    WeightMinimum=0
    Weight=@n22.OutputParam
   )
   > n23 = &RigUnit_SphericalPoseReader(
    ActiveRegionScaleFactors=(PositiveWidth=1,NegativeWidth=1,PositiveHeight=1,NegativeHeight=0.5)
    ActiveRegionSize=0.15
    Debug=(bDrawDebug=false,bDraw2D=False,bDrawLocalAxes=False,DebugScale=25,DebugSegments=20,DebugThickness=0.25)
    DriverAxis=(X=-1,Y=0,Z=0)
    DriverItem=(Type=Bone,Name="upperarm_r")
    FalloffRegionScaleFactors=(PositiveWidth=2,NegativeWidth=1.5,PositiveHeight=1.5,NegativeHeight=1)
    FalloffSize=0.3
    FlipHeightScaling=False
    FlipWidthScaling=False
    OptionalParentItem=(Type=Bone,Name="None")
    RotationOffset=(X=-140,Y=40,Z=-1.7831)
   )
   n24 = &RigVMFunction_AlphaInterp(InterpSpeedDecreasing=20, InterpSpeedIncreasing=20, Scale=1, bClampResult=False, bInterpResult=True, Value=@n23.OutputParam)
   > n25 = &RigUnit_SetCurveValue(Curve=upperarm_r_up_SPR, Value=@n24.Result)
   n26 = &RigVMFunction_ControlFlowBranch(Condition=@n1.CustomIdleIsPlaying, ExecuteContext=@n25.ExecutePin)
   RigVMVariableNode(Value=@n24.Result, ExecuteContext=@n26.False) {variable=FEAdjustAlphaRight}
   > n27 = &RigUnit_SphericalPoseReader(
    ActiveRegionScaleFactors=(PositiveWidth=1,NegativeWidth=1,PositiveHeight=1,NegativeHeight=0.5)
    ActiveRegionSize=0.15
    Debug=(bDrawDebug=false,bDraw2D=False,bDrawLocalAxes=False,DebugScale=25,DebugSegments=20,DebugThickness=0.25)
    DriverAxis=(X=-1,Y=0,Z=0)
    DriverItem=(Type=Bone,Name="upperarm_r")
    FalloffRegionScaleFactors=(PositiveWidth=2,NegativeWidth=1.5,PositiveHeight=1.5,NegativeHeight=1)
    FalloffSize=0.5
    FlipHeightScaling=False
    FlipWidthScaling=False
    OptionalParentItem=(Type=Bone,Name="None")
    RotationOffset=(X=-125,Y=40,Z=-1.7831)
   )
   > &RigUnit_ModifyTransforms(
    ItemToModify=((Item=(Type=Bone,Name="deform_shoulderpad_r_MSTR"),Transform=(Rotation=(X=0.5659,Y=-0.095,Z=-0.0894,W=0.8141),Translation=(X=-7.1525,Y=10,Z=6),Scale3D=(X=1,Y=1,Z=1))))
    WeightMinimum=0
    Weight=@n27.OutputParam
   )
   RigVMVariableNode(ExecuteContext=@n26.True) {variable=FEAdjustAlphaRight}
   &RigUnit_ModifyTransforms(
    ItemToModify="((Item=(Type=Bone,Name=\"deform_torso_box_MSTR\"),Transform=(Rotation=(X=0,Y=0,Z=0,W=1),Translation=(X=2,Y=0,Z=0),Scale3D=(X=1,Y=1,Z=1))),(Item=(Type=Bone,Name=\"deform_torso_box_front\"),Transform=(Rotation=(X=0,Y=0,Z=0.1045,W=0.9945),Translation=(X=0,Y=0,Z=0 ...(truncated)"
    WeightMinimum=0
    Weight=RigVMVariableNode() {variable=FEAdjustAlphaRight}.Value
    ExecutePin=@n26.Completed
   )
   > &RigUnit_ModifyTransforms(
    ItemToModify=((Item=(Type=Bone,Name="deform_pelvis_box_MSTR"),Transform=(Rotation=(X=0,Y=0,Z=0,W=1),Translation=(X=-0.75,Y=0,Z=0),Scale3D=(X=1,Y=1,Z=1))))
    WeightMinimum=0
    Weight=&RigVMFunction_MathDoubleSub(A=1, B=RigVMVariableNode() {variable=FEAdjustAlphaRight}.Value).Result
   )
   > n28 = &RigUnit_SphericalPoseReader(
    ActiveRegionScaleFactors=(PositiveWidth=1,NegativeWidth=1,PositiveHeight=1,NegativeHeight=1)
    ActiveRegionSize=0.1
    Debug=(bDrawDebug=false,bDraw2D=False,bDrawLocalAxes=False,DebugScale=25,DebugSegments=20,DebugThickness=0.25)
    DriverAxis=(X=-1,Y=0,Z=0)
    DriverItem=(Type=Bone,Name="lowerarm_r")
    FalloffRegionScaleFactors=(PositiveWidth=1,NegativeWidth=1,PositiveHeight=1,NegativeHeight=1)
    FalloffSize=0.2
    FlipHeightScaling=False
    FlipWidthScaling=False
    OptionalParentItem=(Type=Bone,Name="upperarm_r")
    RotationOffset=(X=-90,Y=0,Z=0)
   )
   > &RigUnit_ModifyTransforms(
    ItemToModify=((Item=(Type=Bone,Name="deform_lowerarm_box_top_r"),Transform=(Rotation=(X=0,Y=0,Z=0,W=1),Translation=(X=-3,Y=0,Z=0),Scale3D=(X=1,Y=1,Z=1))))
    WeightMinimum=0
    Weight=@n28.OutputParam
   )
   > n29 = &RigUnit_SphericalPoseReader(
    ActiveRegionScaleFactors=(PositiveWidth=1,NegativeWidth=1,PositiveHeight=1,NegativeHeight=1)
    ActiveRegionSize=0.1
    Debug=(bDrawDebug=false,bDraw2D=False,bDrawLocalAxes=False,DebugScale=25,DebugSegments=20,DebugThickness=0.25)
    DriverAxis=(X=1,Y=0,Z=0)
    DriverItem=(Type=Bone,Name="calf_r")
    FalloffRegionScaleFactors=(PositiveWidth=1,NegativeWidth=1,PositiveHeight=1,NegativeHeight=1)
    FalloffSize=0.4
    FlipHeightScaling=False
    FlipWidthScaling=False
    OptionalParentItem=(Type=Bone,Name="thigh_r")
    RotationOffset=(X=90,Y=0,Z=50)
   )
   > &RigUnit_ModifyTransforms(
    ItemToModify="((Item=(Type=Bone,Name=\"deform_calf_box_top_r\"),Transform=(Rotation=(X=0,Y=0,Z=0,W=-1),Translation=(X=4,Y=0,Z=0),Scale3D=(X=1,Y=1,Z=1))),(Item=(Type=Bone,Name=\"dyn_calf_box_flap_top_r_bk_r_MSTR\"),Transform=(Rotation=(X=0,Y=0.866,Z=0,W=0.5),Translation=(X=0 ...(truncated)"
    WeightMinimum=0
    Weight=@n29.OutputParam
   )
   > n30 = &RigUnit_SphericalPoseReader(
    ActiveRegionScaleFactors=(PositiveWidth=1,NegativeWidth=1,PositiveHeight=1,NegativeHeight=1)
    ActiveRegionSize=0.1
    Debug=(bDrawDebug=false,bDraw2D=False,bDrawLocalAxes=False,DebugScale=25,DebugSegments=20,DebugThickness=0.25)
    DriverAxis=(X=0,Y=-1,Z=0)
    DriverItem=(Type=Bone,Name="foot_r")
    FalloffRegionScaleFactors=(PositiveWidth=1,NegativeWidth=1,PositiveHeight=1,NegativeHeight=1)
    FalloffSize=0.3
    FlipHeightScaling=False
    FlipWidthScaling=False
    OptionalParentItem=(Type=Bone,Name="calf_r")
    RotationOffset=(X=-90,Y=0,Z=-55)
   )
   n31 = &RigVMFunction_AlphaInterp(InterpSpeedDecreasing=20, InterpSpeedIncreasing=20, Scale=1, bClampResult=False, bInterpResult=True, Value=@n30.OutputParam)
   > &RigUnit_ModifyTransforms(
    ItemToModify="((Item=(Type=Bone,Name=\"deform_calf_box_bottom_r\"),Transform=(Rotation=(X=0,Y=0,Z=0,W=1),Translation=(X=-6,Y=0,Z=0),Scale3D=(X=1,Y=1,Z=1))),(Item=(Type=Bone,Name=\"dyn_calf_box_flap_bottom_fr_r_MSTR\"),Transform=(Rotation=(X=0,Y=-0.6088,Z=0,W=0.7934),Transla ...(truncated)"
    WeightMinimum=0
    Weight=@n31.Result
   )
   > n32 = &RigUnit_SphericalPoseReader(
    ActiveRegionScaleFactors=(PositiveWidth=1,NegativeWidth=1,PositiveHeight=1,NegativeHeight=1)
    ActiveRegionSize=0.1
    Debug=(bDrawDebug=false,bDraw2D=False,bDrawLocalAxes=False,DebugScale=25,DebugSegments=20,DebugThickness=0.25)
    DriverAxis=(X=0,Y=-1,Z=0)
    DriverItem=(Type=Bone,Name="foot_r")
    FalloffRegionScaleFactors=(PositiveWidth=1,NegativeWidth=1,PositiveHeight=1,NegativeHeight=1)
    FalloffSize=0.3
    FlipHeightScaling=False
    FlipWidthScaling=False
    OptionalParentItem=(Type=Bone,Name="calf_r")
    RotationOffset=(X=-90,Y=0,Z=55)
   )
   n33 = &RigVMFunction_AlphaInterp(InterpSpeedDecreasing=20, InterpSpeedIncreasing=20, Scale=1, bClampResult=False, bInterpResult=True, Value=@n32.OutputParam)
   > &RigUnit_ModifyTransforms(
    ItemToModify=((Item=(Type=Bone,Name="deform_calf_box_bottom_r"),Transform=(Rotation=(X=0,Y=0,Z=0,W=1),Translation=(X=0,Y=0,Z=0),Scale3D=(X=1,Y=1.1,Z=1.1))))
    WeightMinimum=0
    Weight=@n33.Result
   )
   RigVMFunctionReturnNode(ExecuteContext=@RigVMFunction_Sequence_1.C)
   n34 = &RigUnit_SphericalPoseReader(
    ActiveRegionScaleFactors=(PositiveWidth=1,NegativeWidth=1,PositiveHeight=1,NegativeHeight=1)
    ActiveRegionSize=0.1
    Debug=(bDrawDebug=true,bDraw2D=False,bDrawLocalAxes=False,DebugScale=25,DebugSegments=20,DebugThickness=0.25)
    DriverAxis=(X=1,Y=0,Z=0)
    DriverItem=(Type=Bone,Name="clavicle_l")
    ExecutePin=()
    FalloffRegionScaleFactors=(PositiveWidth=1,NegativeWidth=1,PositiveHeight=1,NegativeHeight=1)
    FalloffSize=0.2
    FlipHeightScaling=False
    FlipWidthScaling=False
    OptionalParentItem=(Type=Bone,Name="None")
    RotationOffset=(X=90,Y=0,Z=-40)
   )
   > &RigUnit_ModifyTransforms(
    ItemToModify=((Item=(Type=Bone,Name="deform_shoulderpad_l"),Transform=(Rotation=(X=0.0858,Y=0.0151,Z=-0.173,W=0.9811),Translation=(X=3.2117,Y=-3.4891,Z=-0.6152),Scale3D=(X=1,Y=1,Z=1))))
    WeightMinimum=0
    Weight=@n34.OutputParam
   )
   &RigVMFunction_AlphaInterp(InterpSpeedDecreasing=20, InterpSpeedIncreasing=20, Scale=1, bClampResult=False, bInterpResult=True, Value=@n13.OutputParam)
  ]}
 New Function_ContainedGraph {nodes [
   RigVMFunctionEntryNode()
   > &RigUnit_ParentConstraint(
    AdvancedSettings=(RotationOrderForFilter=XZY)
    Child=(Type=Bone,Name="deform_torso_batteries_MSTR")
    Parents="((Item=(Type=Bone,Name=\"clavicle_r\"),Weight=0.0031),(Item=(Type=Bone,Name=\"deform_pec_r\"),Weight=0.0079),(Item=(Type=Bone,Name=\"spine_04\"),Weight=0.0084),(Item=(Type=Bone,Name=\"spine_02\"),Weight=0.0085),(Item=(Type=Bone,Name=\"spine_03\"),Weight=0.014),(Item ...(truncated)"
    Weight=1
   )
   > &RigUnit_ParentConstraint(
    AdvancedSettings=(RotationOrderForFilter=XZY)
    Child=(Type=Bone,Name="deform_torso_acc_A_MSTR")
    Parents=((Item=(Type=Bone,Name="deform_torso_box_bottom"),Weight=0.1123),(Item=(Type=Bone,Name="deform_torso_box_MSTR"),Weight=0.1452),(Item=(Type=Bone,Name="deform_torso_box_front"),Weight=0.7426))
    Weight=1
   )
   > &RigUnit_ParentConstraint(
    AdvancedSettings=(RotationOrderForFilter=XZY)
    Child=(Type=Bone,Name="deform_torso_acc_B_MSTR")
    Parents="((Item=(Type=Bone,Name=\"clavicle_l\"),Weight=0.0012),(Item=(Type=Bone,Name=\"spine_02\"),Weight=0.0027),(Item=(Type=Bone,Name=\"deform_pec_l\"),Weight=0.0035),(Item=(Type=Bone,Name=\"spine_04\"),Weight=0.0037),(Item=(Type=Bone,Name=\"spine_03\"),Weight=0.0047),(Ite ...(truncated)"
    Weight=1
   )
   > &RigUnit_ParentConstraint(
    AdvancedSettings=(RotationOrderForFilter=XZY)
    Child=(Type=Bone,Name="deform_torso_acc_C_MSTR")
    Parents="((Item=(Type=Bone,Name=\"spine_05\"),Weight=0.0056),(Item=(Type=Bone,Name=\"spine_03\"),Weight=0.0063),(Item=(Type=Bone,Name=\"spine_04\"),Weight=0.0083),(Item=(Type=Bone,Name=\"deform_torso_box_l\"),Weight=0.0095),(Item=(Type=Bone,Name=\"deform_pec_l\"),Weight=0.01 ...(truncated)"
    Weight=1
   )
   > &RigUnit_ParentConstraint(
    AdvancedSettings=(RotationOrderForFilter=XZY)
    Child=(Type=Bone,Name="deform_torso_acc_D_MSTR")
    Parents="((Item=(Type=Bone,Name=\"clavicle_r\"),Weight=0.0005),(Item=(Type=Bone,Name=\"spine_03\"),Weight=0.0007),(Item=(Type=Bone,Name=\"deform_torso_box_bottom\"),Weight=0.0009),(Item=(Type=Bone,Name=\"spine_05\"),Weight=0.0009),(Item=(Type=Bone,Name=\"deform_pec_r\"),Weig ...(truncated)"
    Weight=1
   )
   > &RigUnit_ParentConstraint(
    AdvancedSettings=(RotationOrderForFilter=XZY)
    Child=(Type=Bone,Name="deform_torso_acc_E_MSTR")
    Parents="((Item=(Type=Bone,Name=\"clavicle_r\"),Weight=0.0001),(Item=(Type=Bone,Name=\"spine_03\"),Weight=0.0002),(Item=(Type=Bone,Name=\"deform_pec_r\"),Weight=0.0002),(Item=(Type=Bone,Name=\"spine_05\"),Weight=0.0003),(Item=(Type=Bone,Name=\"spine_04\"),Weight=0.0004),(Ite ...(truncated)"
    Weight=1
   )
   > &RigUnit_ParentConstraint(
    AdvancedSettings=(RotationOrderForFilter=XZY)
    Child=(Type=Bone,Name="deform_torso_acc_F_MSTR")
    Parents=((Item=(Type=Bone,Name="deform_torso_box_MSTR"),Weight=0.1629),(Item=(Type=Bone,Name="deform_torso_box_front"),Weight=0.8371))
    Weight=1
   )
   > &RigUnit_ParentConstraint(
    AdvancedSettings=(RotationOrderForFilter=XZY)
    Child=(Type=Bone,Name="deform_torso_acc_G_MSTR")
    Parents="((Item=(Type=Bone,Name=\"spine_02\"),Weight=0.0002),(Item=(Type=Bone,Name=\"clavicle_l\"),Weight=0.0003),(Item=(Type=Bone,Name=\"spine_03\"),Weight=0.0004),(Item=(Type=Bone,Name=\"spine_05\"),Weight=0.0006),(Item=(Type=Bone,Name=\"deform_pec_l\"),Weight=0.0008),(Ite ...(truncated)"
    Weight=1
   )
   > &RigUnit_ParentConstraint(
    AdvancedSettings=(RotationOrderForFilter=XZY)
    Child=(Type=Bone,Name="deform_tape_back_01_MSTR")
    Parents=((Item=(Type=Bone,Name="deform_torso_box_MSTR"),Weight=1.0))
    Weight=1
   )
   > &RigUnit_ParentConstraint(
    AdvancedSettings=(RotationOrderForFilter=XZY)
    Child=(Type=Bone,Name="deform_tape_back_02_MSTR")
    Parents=((Item=(Type=Bone,Name="deform_torso_box_MSTR"),Weight=1.0))
    Weight=1
   )
   > &RigUnit_ParentConstraint(
    AdvancedSettings=(RotationOrderForFilter=XZY)
    Child=(Type=Bone,Name="deform_tape_back_03_MSTR")
    Parents=((Item=(Type=Bone,Name="deform_torso_box_MSTR"),Weight=1.0))
    Weight=1
   )
   > &RigUnit_ParentConstraint(
    AdvancedSettings=(RotationOrderForFilter=XZY)
    Child=(Type=Bone,Name="deform_torso_acc_back_MSTR")
    Parents=((Item=(Type=Bone,Name="deform_torso_box_MSTR"),Weight=1.0))
    Weight=1
   )
   > &RigUnit_ParentConstraint(
    AdvancedSettings=(RotationOrderForFilter=XZY)
    Child=(Type=Bone,Name="deform_torso_batteries_back_r_MSTR")
    Parents=((Item=(Type=Bone,Name="deform_torso_box_MSTR"),Weight=1.0))
    Weight=1
   )
   > &RigUnit_ParentConstraint(
    AdvancedSettings=(RotationOrderForFilter=XZY)
    Child=(Type=Bone,Name="deform_torso_batteries_back_l_MSTR")
    Parents=((Item=(Type=Bone,Name="deform_torso_box_MSTR"),Weight=1.0))
    Weight=1
   )
   > &RigUnit_ParentConstraint(
    AdvancedSettings=(RotationOrderForFilter=XZY)
    Child=(Type=Bone,Name="deform_arm_remote_MSTR")
    Parents=((Item=(Type=Bone,Name="deform_lowerarm_box_bottom_r"),Weight=0.2464),(Item=(Type=Bone,Name="deform_lowerarm_box_top_r"),Weight=0.7536))
    Weight=1
   )
   > &RigUnit_ParentConstraint(
    AdvancedSettings=(RotationOrderForFilter=XZY)
    Child=(Type=Bone,Name="deform_arm_batteries_MSTR")
    Parents=((Item=(Type=Bone,Name="deform_lowerarm_box_bottom_l"),Weight=0.0105),(Item=(Type=Bone,Name="deform_lowerarm_box_top_l"),Weight=0.9895))
    Weight=1
   )
   > &RigUnit_ParentConstraint(
    AdvancedSettings=(RotationOrderForFilter=XZY)
    Child=(Type=Bone,Name="deform_power_strip_MSTR")
    Parents="((Item=(Type=Bone,Name=\"spine_03\"),Weight=0.0007),(Item=(Type=Bone,Name=\"spine_01\"),Weight=0.0009),(Item=(Type=Bone,Name=\"spine_02\"),Weight=0.0009),(Item=(Type=Bone,Name=\"deform_pelvis_box_r\"),Weight=0.0018),(Item=(Type=Bone,Name=\"pelvis\"),Weight=0.0037),( ...(truncated)"
    Weight=1
   )
   > &RigUnit_ParentConstraint(
    AdvancedSettings=(RotationOrderForFilter=XZY)
    Child=(Type=Bone,Name="deform_squeeze_tubes_MSTR")
    Parents=((Item=(Type=Bone,Name="pelvis"),Weight=0.0002),(Item=(Type=Bone,Name="deform_pelvis_box_front"),Weight=0.0087),(Item=(Type=Bone,Name="deform_pelvis_box_back"),Weight=0.064),(Item=(Type=Bone,Name="deform_pelvis_box_l"),Weight=0.9272))
    Weight=1
   )
   > &RigUnit_ParentConstraint(
    AdvancedSettings=(RotationOrderForFilter=XZY)
    Child=(Type=Bone,Name="deform_switch_MSTR")
    Parents="((Item=(Type=Bone,Name=\"upperarm_twist_01_r\"),Weight=0.0002),(Item=(Type=Bone,Name=\"clavicle_r\"),Weight=0.0003),(Item=(Type=Bone,Name=\"spine_01\"),Weight=0.0003),(Item=(Type=Bone,Name=\"spine_04\"),Weight=0.0004),(Item=(Type=Bone,Name=\"spine_03\"),Weight=0.000 ...(truncated)"
    Weight=1
   )
   > &RigUnit_ParentConstraint(
    AdvancedSettings=(InterpolationType=Average,RotationOrderForFilter=XZY)
    Child=(Type=Bone,Name="deform_cartman_hand_l")
    Parents=((Item=(Type=Bone,Name="deform_torso_box_front"),Weight=1))
    Weight=1
   )
   > &RigUnit_ParentConstraint(
    AdvancedSettings=(InterpolationType=Average,RotationOrderForFilter=XZY)
    Child=(Type=Bone,Name="deform_cartman_hand_r")
    Parents=((Item=(Type=Bone,Name="deform_torso_box_front"),Weight=1))
    Weight=1
   )
   > &RigUnit_ParentConstraint(
    AdvancedSettings=(RotationOrderForFilter=XZY)
    Child=(Type=Bone,Name="deform_helmet_wire_bttm_MSTR")
    Parents="((Item=(Type=Bone,Name=\"deform_pec_l\"),Weight=0.0109),(Item=(Type=Bone,Name=\"spine_05\"),Weight=0.012),(Item=(Type=Bone,Name=\"upperarm_twist_01_l\"),Weight=0.0136),(Item=(Type=Bone,Name=\"deform_upperarm_up_l\"),Weight=0.0139),(Item=(Type=Bone,Name=\"deform_upp ...(truncated)"
    Weight=1
   )
   > RigVMFunctionReturnNode()
  ]}
 New Function_ContainedGraph {nodes [
   n1 = RigVMFunctionEntryNode()
   RigVMFunction_ControlFlowBranch = &RigVMFunction_ControlFlowBranch(ExecuteContext=@n1.ExecuteContext, Condition=@n1.isCrouchMoving)
   AlphaInterp = &RigVMFunction_AlphaInterp(InterpSpeedDecreasing=10, InterpSpeedIncreasing=10, Scale=1, bClampResult=False, Value=RigVMVariableNode() {variable=crouchMoveAlpha}.Value)
   &RigUnit_ModifyTransforms(
    ItemToModify=((Item=(Type=Bone,Name="deform_torso_box_r"),Transform=(Rotation=(X=0,Y=0,Z=0,W=1),Translation=(X=4,Y=0,Z=0),Scale3D=(X=1,Y=1,Z=1))))
    WeightMinimum=0
    Weight=@AlphaInterp.Result
    ExecutePin=@RigVMFunction_ControlFlowBranch.Completed
   )
   > RigVMFunctionReturnNode()
   RigVMVariableNode(ExecuteContext=@RigVMFunction_ControlFlowBranch.False) {variable=crouchMoveAlpha}
   RigVMVariableNode(Value=1, ExecuteContext=@RigVMFunction_ControlFlowBranch.True) {variable=crouchMoveAlpha}
  ]}
 RigVMFunctionLibrary {nodes [
   Crouch Deform = RigVMCollapseNode(ExecuteContext=(), isCrouchMoving=False)
   Deform Accessories = RigVMCollapseNode(ExecuteContext=())
   Deform Boxes = RigVMCollapseNode(CustomIdleIsPlaying=False, ExecuteContext=())
   Deform Rig Init Transforms = RigVMCollapseNode(ExecuteContext=())
  ]}
 RigVMFunction_Sequence_2_ContainedGraph {nodes [
   n1 = RigVMFunctionEntryNode()
   n2 = &RigVMFunction_Sequence(ExecuteContext=@n1.ExecuteContext)
   n3 = &RigVMFunction_Sequence(ExecuteContext=@n2.B)
   RigVMFunctionReturnNode(A=@n2.A, B=@n3.A, C=@n3.B)
  ]}
 RigVMModel {nodes [
   RigUnit_BeginExecution = &RigUnit_BeginExecution()
   Deform Rig Init Transforms = RigVMFunctionReferenceNode(ExecuteContext=@RigUnit_BeginExecution.ExecutePin)
   > Deform Boxes = RigVMFunctionReferenceNode(CustomIdleIsPlaying=RigVMVariableNode() {variable=isCustomIdlePlaying}.Value)
   > Crouch Deform = RigVMFunctionReferenceNode(isCrouchMoving=RigVMVariableNode() {variable=isCrouchMoving}.Value)
   > Deform Accessories = RigVMFunctionReferenceNode()
  ]}
]
hierarchy [
 root {loc=<0 0 0>, rot=<0 0 0 1>}
 pelvis {parent=root, loc=<0 -0.7321 92.4218>, rot=<0.0127 0.707 -0.0127 -0.707>}
 spine_01 {parent=pelvis, loc=<5.6002 -0.5128 0>, rot=<0 0 0.034 -0.9994>}
 spine_02 {parent=spine_01, loc=<8.0756 0.0139 0>, rot=<0 0 0.0238 -0.9997>}
 spine_03 {parent=spine_02, loc=<7.0395 0 0>, rot=<0 0 0.0149 -0.9999>}
 spine_04 {parent=spine_03, loc=<8.1023 0 0>, rot=<0 0 -0.0132 -0.9999>}
 spine_05 {parent=spine_04, loc=<10.0141 0 0>, rot=<0 0 -0.0015 -1>}
 clavicle_l {parent=spine_05, loc=<2.5856 -0.2866 -4.3777>, rot=<0.0465 -0.7289 -0.009 -0.6829>}
 upperarm_l {parent=clavicle_l, loc=<18.2264 0.1342 -0.001>, rot=<0.0253 -0.2693 0.0013 -0.9627>}
 lowerarm_l {parent=upperarm_l, loc=<29.0071 0.1686 -1.239>, rot=<0.0391 -0.0099 -0.2453 -0.9686>}
 hand_l {parent=lowerarm_l, loc=<25.8607 0.222 1.0274>, rot=<-0.7193 -0.0988 0.1193 -0.6773>}
 thumb_01_l {parent=hand_l, loc=<1.3118 -0.3191 -2.459>, rot=<0.6546 -0.4042 0.0664 -0.6354>}
 thumb_02_l {parent=thumb_01_l, loc=<5.9336 -0.0448 0.0401>, rot=<-0.0037 0.0213 0.298 -0.9543>}
 thumb_03_l {parent=thumb_02_l, loc=<2.2223 0.3682 -0.0907>, rot=<0.0118 -0.013 0.0066 -0.9998>}
 index_metacarpal_l {parent=hand_l, loc=<1.6535 0.0183 -2.96>, rot=<0.0158 -0.0124 -0.0173 -0.9996>}
 index_01_l {parent=index_metacarpal_l, loc=<8.3008 0.0357 -0.0524>, rot=<-0.0044 0.0485 0.0924 -0.9945>}
 index_02_l {parent=index_01_l, loc=<3.3587 -0.3004 0.0452>, rot=<-0.002 0.0027 0.3225 -0.9466>}
 index_03_l {parent=index_02_l, loc=<2.5578 -0.123 -0.0252>, rot=<-0.002 0.0067 0.2558 -0.9667>}
 middle_metacarpal_l {parent=hand_l, loc=<2.1856 -0.1725 -0.5984>, rot=<0.006 0.0168 -0.0329 -0.9993>}
 middle_01_l {parent=middle_metacarpal_l, loc=<7.3275 -0.2447 0.0129>, rot=<-0.0208 0.0042 0.0858 -0.9961>}
 middle_02_l {parent=middle_01_l, loc=<4.0123 0.1401 0.1175>, rot=<-0.0218 0.0258 0.3512 -0.9357>}
 middle_03_l {parent=middle_02_l, loc=<2.6077 0.0534 0.1373>, rot=<0.0082 0.0188 0.1708 -0.9851>}
 ring_metacarpal_l {parent=hand_l, loc=<2.2074 -0.3625 1.3638>, rot=<-0.0284 0.0722 -0.0458 -0.9959>}
 ring_01_l {parent=ring_metacarpal_l, loc=<7.0295 -0.6197 -0.1446>, rot=<-0.0489 -0.0252 0.1892 -0.9804>}
 ring_02_l {parent=ring_01_l, loc=<3.6174 0.4647 -0.1027>, rot=<0 0 0.2806 -0.9598>}
 ring_03_l {parent=ring_02_l, loc=<2.4152 0.114 0.0199>, rot=<0 0 0.1189 -0.9929>}
 pinky_metacarpal_l {parent=hand_l, loc=<1.7616 -0.9753 3.3814>, rot=<-0.0397 0.0814 -0.069 -0.9935>}
 pinky_01_l {parent=pinky_metacarpal_l, loc=<7.1709 -0.4831 -0.1094>, rot=<-0.0654 -0.0035 0.2242 -0.9723>}
 pinky_02_l {parent=pinky_01_l, loc=<2.8274 0.0893 -0.1043>, rot=<0.0037 0.0011 0.2854 -0.9584>}
 pinky_03_l {parent=pinky_02_l, loc=<2.2046 0.0307 -0.0078>, rot=<-0.0002 0.0098 0.1445 -0.9895>}
 deform_wrist_dn_base_l {parent=hand_l, loc=<0 -3.4173 0>, rot=<0 0 0 -1>}
 deform_wrist_dn_l {parent=deform_wrist_dn_base_l, loc=<0 0 0>, rot=<0 0 0 1>}
 deform_wrist_up_base_l {parent=hand_l, loc=<0 2.7805 0>, rot=<0 0 0 -1>}
 deform_wrist_up_l {parent=deform_wrist_up_base_l, loc=<0 0 0>, rot=<0 0 0 1>}
 lowerarm_twist_02_l {parent=lowerarm_l, loc=<9.6199 0 0>, rot=<0 0 0 -1>}
 lowerarm_twist_01_l {parent=lowerarm_l, loc=<17.8549 0 0>, rot=<0 0 0 -1>}
 deform_elbow_in_base_l {parent=lowerarm_l, loc=<0.2112 5.645 0.8274>, rot=<0 0 0 -1>}
 deform_elbow_in_l {parent=deform_elbow_in_base_l, loc=<0 0 0>, rot=<0 0 0 1>}
 deform_elbow_out_base_l {parent=lowerarm_l, loc=<0 -7.5129 0>, rot=<0 0 0 -1>}
 deform_elbow_out_l {parent=deform_elbow_out_base_l, loc=<0 0 0>, rot=<0 0 0 1>}
 deform_lowerarm_box_l_MSTR {parent=lowerarm_l, loc=<14.7878 0 0>, rot=<0 0 0 -1>}
 deform_lowerarm_box_bottom_l {parent=deform_lowerarm_box_l_MSTR, loc=<8.2944 0 0>, rot=<0 0 0 -1>}
 dyn_lowerarm_flap_fr_l_MSTR {parent=deform_lowerarm_box_bottom_l, loc=<-0.0039 5.5252 -0.1324>, rot=<0.5936 0.3531 -0.3722 -0.62>}
 dyn_lowerarm_flap_fr_l {parent=dyn_lowerarm_flap_fr_l_MSTR, loc=<0 0 0>, rot=<0 0 0 1>}
 dyn_lowerarm_flap_in_l_MSTR {parent=deform_lowerarm_box_bottom_l, loc=<-0.3495 0.0409 -5.3103>, rot=<0.8319 -0.018 -0.5537 -0.031>}
 dyn_lowerarm_flap_in_l {parent=dyn_lowerarm_flap_in_l_MSTR, loc=<0 0 0>, rot=<0 0 0 1>}
 dyn_lowerarm_flap_back_l_MSTR {parent=deform_lowerarm_box_bottom_l, loc=<-0.3083 -4.7066 0.063>, rot=<-0.6119 0.3157 0.3376 -0.6418>}
 dyn_lowerarm_flap_back_l {parent=dyn_lowerarm_flap_back_l_MSTR, loc=<0 0 0>, rot=<0 0 0 1>}
 dyn_lowerarm_flap_out_l_MSTR {parent=deform_lowerarm_box_bottom_l, loc=<0.1037 0.8312 5.1785>, rot=<0.0494 0.3269 -0.086 -0.9398>}
 dyn_lowerarm_flap_out_l {parent=dyn_lowerarm_flap_out_l_MSTR, loc=<0 0 0>, rot=<0 0 0 1>}
 deform_lowerarm_box_top_l {parent=deform_lowerarm_box_l_MSTR, loc=<-15.4501 0 0>, rot=<0 0 0 1>}
 deform_arm_batteries_MSTR {parent=deform_lowerarm_box_top_l, loc=<3.2356 1.9143 11.2638>, rot=<0.0394 0.0008 0.0454 -0.9982>}
 deform_arm_batteries {parent=deform_arm_batteries_MSTR, loc=<0 0 0>, rot=<0 0 0 1>}
 upperarm_twist_01_l {parent=upperarm_l, loc=<7.9796 0.0843 -0.6195>, rot=<0 0 0 -1>}
 deform_upperarm_bk_base_l {parent=upperarm_twist_01_l, loc=<-6.8983 -13.3826 -1.3258>, rot=<0 0 0 1>}
 deform_upperarm_bk_l {parent=deform_upperarm_bk_base_l, loc=<0 0 0>, rot=<0 0 0 1>}
 deform_upperarm_fr_base_l {parent=upperarm_twist_01_l, loc=<-8.3487 10.7117 -2.1455>, rot=<0 0 0 1>}
 deform_upperarm_fr_l {parent=deform_upperarm_fr_base_l, loc=<0 0 0>, rot=<0 0 0 1>}
 deform_upperarm_up_base_l {parent=upperarm_twist_01_l, loc=<-11.2012 -1.2234 8.7784>, rot=<0 0 0 1>}
 deform_upperarm_up_l {parent=deform_upperarm_up_base_l, loc=<0 0 0>, rot=<0 0 0 1>}
 deform_shoulderpad_l_MSTR {parent=upperarm_twist_01_l, loc=<-2.9401 2.8434 9.2756>, rot=<0 0 0 1>}
 deform_shoulderpad_l {parent=deform_shoulderpad_l_MSTR, loc=<0 0 0>, rot=<0 0 0 1>}
 upperarm_twist_02_l {parent=upperarm_l, loc=<19.5059 0.0843 -0.6195>, rot=<0 0 0 -1>}
 deform_bicep_base_l {parent=upperarm_twist_02_l, loc=<0 8.0494 0>, rot=<0 0 0 1>}
 deform_bicep_l {parent=deform_bicep_base_l, loc=<0 0 0>, rot=<0 0 0 1>}
 neck_01 {parent=spine_05, loc=<10.815 1.9196 0>, rot=<0 0 -0.1864 -0.9825>}
 neck_02 {parent=neck_01, loc=<4.7347 0.3045 0>, rot=<0 0 0.0363 -0.9993>}
 head {parent=neck_02, loc=<5.7632 0.5552 0>, rot=<0 0 0.115 -0.9934>}
 clavicle_r {parent=spine_05, loc=<2.5855 -0.2866 4.3777>, rot=<0.7289 0.0465 -0.6829 0.009>}
 upperarm_r {parent=clavicle_r, loc=<-18.2264 -0.1342 0.001>, rot=<0.0253 -0.2693 0.0013 -0.9627>}
 lowerarm_r {parent=upperarm_r, loc=<-29.0071 -0.1686 1.239>, rot=<0.0391 -0.0099 -0.2453 -0.9686>}
 hand_r {parent=lowerarm_r, loc=<-25.8607 -0.222 -1.0274>, rot=<-0.7193 -0.0988 0.1193 -0.6773>}
 thumb_01_r {parent=hand_r, loc=<-1.3118 0.3191 2.459>, rot=<0.6546 -0.4042 0.0664 -0.6354>}
 thumb_02_r {parent=thumb_01_r, loc=<-5.9336 0.0448 -0.0401>, rot=<-0.0037 0.0213 0.298 -0.9543>}
 thumb_03_r {parent=thumb_02_r, loc=<-2.2223 -0.3682 0.0907>, rot=<0.0118 -0.013 0.0066 -0.9998>}
 index_metacarpal_r {parent=hand_r, loc=<-1.6535 -0.0183 2.96>, rot=<0.0158 -0.0124 -0.0173 -0.9996>}
 index_01_r {parent=index_metacarpal_r, loc=<-8.3008 -0.0357 0.0524>, rot=<-0.0044 0.0485 0.0924 -0.9945>}
 index_02_r {parent=index_01_r, loc=<-3.3587 0.3004 -0.0452>, rot=<-0.002 0.0027 0.3225 -0.9466>}
 index_03_r {parent=index_02_r, loc=<-2.5578 0.123 0.0252>, rot=<-0.002 0.0067 0.2558 -0.9667>}
 middle_metacarpal_r {parent=hand_r, loc=<-2.1856 0.1725 0.5984>, rot=<0.006 0.0168 -0.0329 -0.9993>}
 middle_01_r {parent=middle_metacarpal_r, loc=<-7.3275 0.2447 -0.0129>, rot=<-0.0208 0.0042 0.0858 -0.9961>}
 middle_02_r {parent=middle_01_r, loc=<-4.0123 -0.1401 -0.1175>, rot=<-0.0218 0.0258 0.3512 -0.9357>}
 middle_03_r {parent=middle_02_r, loc=<-2.6077 -0.0534 -0.1373>, rot=<0.0082 0.0188 0.1708 -0.9851>}
 ring_metacarpal_r {parent=hand_r, loc=<-2.2074 0.3625 -1.3638>, rot=<-0.0284 0.0722 -0.0458 -0.9959>}
 ring_01_r {parent=ring_metacarpal_r, loc=<-7.0295 0.6197 0.1446>, rot=<-0.0489 -0.0252 0.1892 -0.9804>}
 ring_02_r {parent=ring_01_r, loc=<-3.6174 -0.4647 0.1027>, rot=<0 0 0.2806 -0.9598>}
 ring_03_r {parent=ring_02_r, loc=<-2.4152 -0.114 -0.0199>, rot=<0 0 0.1189 -0.9929>}
 pinky_metacarpal_r {parent=hand_r, loc=<-1.7616 0.9753 -3.3814>, rot=<-0.0397 0.0814 -0.069 -0.9935>}
 pinky_01_r {parent=pinky_metacarpal_r, loc=<-7.1709 0.4831 0.1094>, rot=<-0.0654 -0.0035 0.2242 -0.9723>}
 pinky_02_r {parent=pinky_01_r, loc=<-2.8274 -0.0893 0.1043>, rot=<0.0037 0.0011 0.2854 -0.9584>}
 pinky_03_r {parent=pinky_02_r, loc=<-2.2046 -0.0307 0.0078>, rot=<-0.0002 0.0098 0.1445 -0.9895>}
 deform_wrist_dn_base_r {parent=hand_r, loc=<0 3.4172 0>, rot=<0 0 0 -1>}
 deform_wrist_dn_r {parent=deform_wrist_dn_base_r, loc=<0 0 0>, rot=<0 0 0 -1>}
 deform_wrist_up_base_r {parent=hand_r, loc=<-0.0001 -2.7804 0>, rot=<0 0 0 -1>}
 deform_wrist_up_r {parent=deform_wrist_up_base_r, loc=<0 0 0>, rot=<0 0 0 -1>}
 lowerarm_twist_02_r {parent=lowerarm_r, loc=<-9.6199 0 0>, rot=<0 0 0 -1>}
 lowerarm_twist_01_r {parent=lowerarm_r, loc=<-17.8549 0 0>, rot=<0 0 0 -1>}
 deform_elbow_in_base_r {parent=lowerarm_r, loc=<-0.2114 -5.6449 -0.827>, rot=<0 0 0 -1>}
 deform_elbow_in_r {parent=deform_elbow_in_base_r, loc=<0 0 0>, rot=<0 0 0 1>}
 deform_elbow_out_base_r {parent=lowerarm_r, loc=<-0.0001 7.5129 0.0002>, rot=<0 0 0 -1>}
 deform_elbow_out_r {parent=deform_elbow_out_base_r, loc=<0 0 0>, rot=<0 0 0 1>}
 deform_lowerarm_box_r_MSTR {parent=lowerarm_r, loc=<-14.7877 0 0>, rot=<0 0 0 -1>}
 deform_lowerarm_box_bottom_r {parent=deform_lowerarm_box_r_MSTR, loc=<-8.1837 0 0>, rot=<0 0 0 1>}
 dyn_lowerarm_flap_out_r_MSTR {parent=deform_lowerarm_box_bottom_r, loc=<-0.2143 -0.8313 -5.1787>, rot=<0.0659 -0.9351 0.1067 -0.3315>}
 dyn_lowerarm_flap_out_r {parent=dyn_lowerarm_flap_out_r_MSTR, loc=<0 0 0>, rot=<0 0 0 1>}
 dyn_lowerarm_flap_fr_r_MSTR {parent=deform_lowerarm_box_bottom_r, loc=<-0.1068 -5.5253 0.132>, rot=<0.3722 -0.62 0.5937 -0.353>}
 dyn_lowerarm_flap_fr_r {parent=dyn_lowerarm_flap_fr_r_MSTR, loc=<0 0 0>, rot=<0 0 0 1>}
 dyn_lowerarm_flap_back_r_MSTR {parent=deform_lowerarm_box_bottom_r, loc=<0.1975 4.7066 -0.063>, rot=<-0.3431 -0.6524 -0.6006 -0.3098>}
 dyn_lowerarm_flap_back_r {parent=dyn_lowerarm_flap_back_r_MSTR, loc=<0 0 0>, rot=<0 0 0 1>}
 dyn_lowerarm_flap_in_r_MSTR {parent=deform_lowerarm_box_bottom_r, loc=<0.2387 -0.0409 5.3103>, rot=<0.554 -0.0092 0.8325 0.0035>}
 dyn_lowerarm_flap_in_r {parent=dyn_lowerarm_flap_in_r_MSTR, loc=<0 0 0>, rot=<0 0 0 1>}
 deform_lowerarm_box_top_r {parent=deform_lowerarm_box_r_MSTR, loc=<13.5193 0 0>, rot=<0 0 0 1>}
 deform_arm_remote_MSTR {parent=deform_lowerarm_box_top_r, loc=<-10.5363 -1.813 -9.3383>, rot=<-0.0186 -0.9923 0.0665 0.1031>}
 deform_arm_remote {parent=deform_arm_remote_MSTR, loc=<0 0 0>, rot=<0 0 0 1>}
 upperarm_twist_01_r {parent=upperarm_r, loc=<-7.9796 -0.0843 0.6195>, rot=<0 0 0 -1>}
 deform_upperarm_bk_base_r {parent=upperarm_twist_01_r, loc=<6.8985 13.3826 1.3257>, rot=<0 0 0 1>}
 deform_upperarm_bk_r {parent=deform_upperarm_bk_base_r, loc=<0 0 0>, rot=<0 0 0 1>}
 deform_upperarm_fr_base_r {parent=upperarm_twist_01_r, loc=<8.3485 -10.7117 2.1456>, rot=<0 0 0 1>}
 deform_upperarm_fr_r {parent=deform_upperarm_fr_base_r, loc=<0 0 0>, rot=<0 0 0 1>}
 deform_upperarm_up_base_r {parent=upperarm_twist_01_r, loc=<11.2015 1.2234 -8.7788>, rot=<0 0 0 1>}
 deform_upperarm_up_r {parent=deform_upperarm_up_base_r, loc=<0 0 0>, rot=<0 0 0 1>}
 deform_shoulderpad_r_MSTR {parent=upperarm_twist_01_r, loc=<2.9401 -2.8434 -9.2756>, rot=<0 0 0 1>}
 deform_shoulderpad_r {parent=deform_shoulderpad_r_MSTR, loc=<0 0 0>, rot=<0 0 0 1>}
 dyn_shoulder_cord_MSTR {parent=deform_shoulderpad_r, loc=<7.4418 -2.894 -0.8696>, rot=<-0.0646 -0.9702 0.0413 -0.2298>}
 dyn_shoulder_cord_01 {parent=dyn_shoulder_cord_MSTR, loc=<0 0 0>, rot=<0 0 0 1>}
 dyn_shoulder_cord_02 {parent=dyn_shoulder_cord_01, loc=<7.9304 0 0>, rot=<0 0 0 1>}
 upperarm_twist_02_r {parent=upperarm_r, loc=<-19.5059 -0.0843 0.6195>, rot=<0 0 0 -1>}
 deform_bicep_base_r {parent=upperarm_twist_02_r, loc=<0.0002 -8.0494 -0.0003>, rot=<0 0 0 1>}
 deform_bicep_r {parent=deform_bicep_base_r, loc=<0 0 0>, rot=<0 0 0 1>}
 attach_cape {parent=spine_05, loc=<9.0697 -11.8367 -0.0014>, rot=<0.0613 -0.7043 -0.0616 -0.7046>}
 attach_backpack {parent=spine_05, loc=<0.0198 -13.6094 -0.0053>, rot=<-0.0096 -0.7069 0.0092 -0.7072>}
 Backpack_BR {parent=spine_05, type=Null, loc=<-0.1228 -19.9024 0>, rot=<0 -0.7071 0 -0.7071>}
 cape {parent=spine_05, type=Null, loc=<17.701 -19.1158 0.6492>, rot=<-0.0304 0.7064 0.0261 0.7066>}
 deform_torso_box_MSTR {parent=spine_05, loc=<-2.8479 2.5282 -0.0313>, rot=<0 0 -0.0401 -0.9992>}
 deform_torso_box_bottom {parent=deform_torso_box_MSTR, loc=<-17.7247 0 0>, rot=<0 0 0 1>}
 deform_torso_box_front {parent=deform_torso_box_MSTR, loc=<0 19.9578 0>, rot=<-0.5 -0.5 -0.5 -0.5>}
 deform_torso_box_l {parent=deform_torso_box_MSTR, loc=<-11.3746 0 -20>, rot=<0 -0.7071 0 -0.7071>}
 deform_torso_box_r {parent=deform_torso_box_MSTR, loc=<-11.3746 0 20>, rot=<0.7071 0 0.7071 0>}
 deform_torso_acc_A_MSTR {parent=deform_torso_box_MSTR, loc=<-6.86 20.2627 0.6985>, rot=<0 -0.7071 0 -0.7071>}
 deform_torso_acc_A {parent=deform_torso_acc_A_MSTR, loc=<0 0 0>, rot=<0 0 0 1>}
 deform_torso_acc_B_MSTR {parent=deform_torso_box_MSTR, loc=<-6.8602 20.6308 -7.8786>, rot=<0 -0.7071 0 -0.7071>}
 deform_torso_acc_B {parent=deform_torso_acc_B_MSTR, loc=<0 0 0>, rot=<0 0 0 1>}
 deform_torso_acc_C_MSTR {parent=deform_torso_box_MSTR, loc=<0.4273 20.9014 -14.6013>, rot=<0 -0.7071 0 -0.7071>}
 deform_torso_acc_C {parent=deform_torso_acc_C_MSTR, loc=<0 0 0>, rot=<0 0 0 1>}
 deform_torso_acc_D_MSTR {parent=deform_torso_box_MSTR, loc=<5.2945 20.6027 8.7266>, rot=<-0.5 -0.5 -0.5 -0.5>}
 deform_torso_acc_D {parent=deform_torso_acc_D_MSTR, loc=<0 0 0>, rot=<0 0 0 1>}
 deform_torso_acc_E_MSTR {parent=deform_torso_box_MSTR, loc=<5.4431 20.6975 3.8797>, rot=<-0.5 -0.5 -0.5 -0.5>}
 deform_torso_acc_E {parent=deform_torso_acc_E_MSTR, loc=<0 0 0>, rot=<0 0 0 1>}
 deform_torso_acc_F_MSTR {parent=deform_torso_box_MSTR, loc=<5.4147 20.9661 -1.3311>, rot=<-0.5 -0.5 -0.5 -0.5>}
 deform_torso_acc_F {parent=deform_torso_acc_F_MSTR, loc=<0 0 0>, rot=<0 0 0 1>}
 deform_torso_acc_G_MSTR {parent=deform_torso_box_MSTR, loc=<5.4669 20.8901 -6.9235>, rot=<-0.5 -0.5 -0.5 -0.5>}
 deform_torso_acc_G {parent=deform_torso_acc_G_MSTR, loc=<0 0 0>, rot=<0 0 0 1>}
 deform_tape_back_01_MSTR {parent=deform_torso_box_MSTR, loc=<5.3966 -20.0744 -13.2232>, rot=<0.5 -0.5 0.5 -0.5>}
 deform_tape_back_01 {parent=deform_tape_back_01_MSTR, loc=<0 0 0>, rot=<0 0 0 1>}
 deform_tape_back_02_MSTR {parent=deform_torso_box_MSTR, loc=<-12.2076 -20.6914 -1.3103>, rot=<0.5 -0.5 0.5 -0.5>}
 deform_tape_back_02 {parent=deform_tape_back_02_MSTR, loc=<0 0 0>, rot=<0 0 0 1>}
 deform_tape_back_03_MSTR {parent=deform_torso_box_MSTR, loc=<-0.1292 -20.2944 15.2196>, rot=<0.5 -0.5 0.5 -0.5>}
 deform_tape_back_03 {parent=deform_tape_back_03_MSTR, loc=<0 0 0>, rot=<0 0 0 1>}
 deform_torso_acc_back_MSTR {parent=deform_torso_box_MSTR, loc=<-3.2559 -21.0402 -0.2402>, rot=<0.5 -0.5 0.5 -0.5>}
 deform_torso_acc_back {parent=deform_torso_acc_back_MSTR, loc=<0 0 0>, rot=<0 0 0 1>}
 deform_torso_batteries_back_r_MSTR {parent=deform_torso_box_MSTR, loc=<11.4479 -22.4084 10.3373>, rot=<0.5 -0.5 0.5 -0.5>}
 deform_torso_batteries_back_r {parent=deform_torso_batteries_back_r_MSTR, loc=<0 0 0>, rot=<0 0 0 1>}
 deform_torso_batteries_back_l_MSTR {parent=deform_torso_box_MSTR, loc=<11.4479 -22.4084 -10.2747>, rot=<0.5 -0.5 0.5 -0.5>}
 deform_torso_batteries_back_l {parent=deform_torso_batteries_back_l_MSTR, loc=<0 0 0>, rot=<0 0 0 1>}
 deform_torso_batteries_MSTR {parent=deform_torso_box_MSTR, loc=<-10.8736 22.3074 13.1814>, rot=<-0.5 -0.5 -0.5 -0.5>}
 deform_torso_batteries {parent=deform_torso_batteries_MSTR, loc=<0 0 0>, rot=<0 0 0 1>}
 deform_cartman_hand_l {parent=spine_05, loc=<13.5215 25.1112 -11.3236>, rot=<0.0284 -0.7065 -0.0284 -0.7065>}
 deform_cartman_hand_r {parent=spine_05, loc=<13.4081 25.4613 11.3269>, rot=<0.0284 -0.7065 -0.0284 -0.7065>}
 deform_helmet_wire_bttm_MSTR {parent=spine_05, loc=<9.2225 20.7648 -18.0714>, rot=<0.0284 -0.7065 -0.0284 -0.7065>}
 deform_helmet_wire_bttm {parent=deform_helmet_wire_bttm_MSTR, loc=<0 0 0>, rot=<0 0 0 1>}
 deform_pec_aim_base_l {parent=spine_04, loc=<5.0227 13.6719 -2.6508>, rot=<0.0672 -0.5201 0.3228 -0.7879>}
 deform_pec_aim_l {parent=deform_pec_aim_base_l, loc=<0 0 0>, rot=<0 0 0 1>}
 deform_pec_base_l {parent=deform_pec_aim_l, loc=<16.3279 0 0>, rot=<0 0 0 1>}
 deform_pec_l {parent=deform_pec_base_l, loc=<0 0 0>, rot=<0 0 0 1>}
 deform_pec_aim_base_r {parent=spine_04, loc=<5.0225 13.6718 2.6508>, rot=<-0.5201 -0.0672 0.7879 0.3228>}
 deform_pec_aim_r {parent=deform_pec_aim_base_r, loc=<0 0 0>, rot=<0 0 0 1>}
 deform_pec_base_r {parent=deform_pec_aim_r, loc=<-16.3279 0 0>, rot=<0 0 0 -1>}
 deform_pec_r {parent=deform_pec_base_r, loc=<0 0 0.0002>, rot=<0 0 0 -1>}
 deform_lat_l_base {parent=spine_03, loc=<2.5561 -10.3716 -17.6918>, rot=<0 0 0 -1>}
 deform_lat_l {parent=deform_lat_l_base, loc=<0 0 0>, rot=<0 0 0 1>}
 deform_lat_r_base {parent=spine_03, loc=<2.5562 -10.3716 17.6918>, rot=<0 0 1 0>}
 deform_lat_r {parent=deform_lat_r_base, loc=<0 0 0>, rot=<0 0 0 -1>}
 thigh_l {parent=pelvis, loc=<-1.8716 -0.0109 -11.9936>, rot=<0.0012 0.0572 0.0155 -0.9982>}
 calf_l {parent=thigh_l, loc=<-41.9178 -0.2246 0.0926>, rot=<0.0006 -0.0161 -0.0391 -0.9991>}
 foot_l {parent=calf_l, loc=<-38.7057 1.2438 0.2452>, rot=<0.0975 -0.0474 0.0317 -0.9936>}
 ball_l {parent=foot_l, loc=<-6.1614 13.1469 0.6807>, rot=<0.0001 -0.0001 -0.7063 -0.7079>}
 calf_twist_02_l {parent=calf_l, loc=<-15.9268 -0.0967 0.0397>, rot=<0.0237 0.0095 -0.0035 -0.9997>}
 calf_twist_01_l {parent=calf_l, loc=<-32.9335 -0.0967 0.0397>, rot=<0.0237 0.0095 -0.0035 -0.9997>}
 deform_knee_in_base_l {parent=calf_l, loc=<1.298 -8.0356 -0.0496>, rot=<0 0 0 -1>}
 deform_knee_in_l {parent=deform_knee_in_base_l, loc=<0 0 0>, rot=<0 0 0 1>}
 deform_knee_out_base_l {parent=calf_l, loc=<0 6.8739 0>, rot=<0 0 0 -1>}
 deform_knee_out_l {parent=deform_knee_out_base_l, loc=<0 0 0>, rot=<0 0 0 1>}
 deform_calf_box_l_MSTR {parent=calf_l, loc=<-21.5 -0.1389 -0.0666>, rot=<0.0237 0.0095 -0.0035 -0.9997>}
 deform_calf_box_bottom_l {parent=deform_calf_box_l_MSTR, loc=<-12.9811 0 0>, rot=<0 0 0 1>}
 dyn_calf_box_flap_bottom_fr_l_MSTR {parent=deform_calf_box_bottom_l, loc=<0.1745 7.8437 0.6186>, rot=<0.4757 0.5171 0.633 0.325>}
 dyn_calf_box_flap_bottom_fr_l {parent=dyn_calf_box_flap_bottom_fr_l_MSTR, loc=<0 0 0>, rot=<0 0 0 1>}
 dyn_calf_box_flap_bottom_in_l_MSTR {parent=deform_calf_box_bottom_l, loc=<1.1481 0.4153 9.574>, rot=<0.3181 -0.0249 0.9477 0.0007>}
 dyn_calf_box_flap_bottom_in_l {parent=dyn_calf_box_flap_bottom_in_l_MSTR, loc=<0 0 0>, rot=<0 0 0 1>}
 dyn_calf_box_flap_bottom_bk_l_MSTR {parent=deform_calf_box_bottom_l, loc=<-0.0865 -7.7238 1.2144>, rot=<0.2895 -0.6811 0.6316 -0.231>}
 dyn_calf_box_flap_bottom_bk_l {parent=dyn_calf_box_flap_bottom_bk_l_MSTR, loc=<0 0 0>, rot=<0 0 0 1>}
 dyn_calf_box_flap_bottom_out_l_MSTR {parent=deform_calf_box_bottom_l, loc=<-0.6104 1.7961 -7.6332>, rot=<-0.0193 -0.9555 0.0055 -0.2943>}
 dyn_calf_box_flap_bottom_out_l {parent=dyn_calf_box_flap_bottom_out_l_MSTR, loc=<0 0 0>, rot=<0 0 0 1>}
 deform_calf_box_top_l {parent=deform_calf_box_l_MSTR, loc=<10.3162 0 0>, rot=<0 0 0 1>}
 dyn_calf_box_flap_top_fr_l_MSTR {parent=deform_calf_box_top_l, loc=<1.0439 8.2096 1.0051>, rot=<-0.7319 -0.2076 -0.2537 -0.5974>}
 dyn_calf_box_flap_top_fr_l {parent=dyn_calf_box_flap_top_fr_l_MSTR, loc=<0 0 0>, rot=<0 0 0 1>}
 dyn_calf_box_flap_top_out_l_MSTR {parent=deform_calf_box_top_l, loc=<0.452 1.0956 -7.1935>, rot=<-0.0541 -0.3411 -0.0003 -0.9385>}
 dyn_calf_box_flap_top_out_l {parent=dyn_calf_box_flap_top_out_l_MSTR, loc=<0 0 0>, rot=<0 0 0 1>}
 dyn_calf_box_flap_top_bk_l_MSTR {parent=deform_calf_box_top_l, loc=<0.1991 -7.1787 0.3565>, rot=<0.6609 -0.2647 0.2524 -0.6553>}
 dyn_calf_box_flap_top_bk_l {parent=dyn_calf_box_flap_top_bk_l_MSTR, loc=<0 0 0>, rot=<0 0 0 1>}
 dyn_calf_box_flap_top_in_l_MSTR {parent=deform_calf_box_top_l, loc=<0.8897 -0.2291 8.3022>, rot=<0.9735 -0.0062 0.2279 -0.0167>}
 dyn_calf_box_flap_top_in_l {parent=dyn_calf_box_flap_top_in_l_MSTR, loc=<0 0 0>, rot=<0 0 0 1>}
 thigh_twist_01_l {parent=thigh_l, loc=<-21.6303 0 0>, rot=<0.0013 0 0 -1>}
 deform_groin_base_l {parent=thigh_l, loc=<-2.1908 8.5661 -1.7442>, rot=<0 0 0 -1>}
 deform_groin_l {parent=deform_groin_base_l, loc=<0 0 0>, rot=<0 0 0 1>}
 thigh_r {parent=pelvis, loc=<-1.8716 -0.0109 11.9936>, rot=<0.0572 -0.0012 0.9982 0.0155>}
 calf_r {parent=thigh_r, loc=<41.9178 0.2246 -0.0926>, rot=<0.0006 -0.0161 -0.0391 -0.9991>}
 foot_r {parent=calf_r, loc=<38.7057 -1.2438 -0.2452>, rot=<0.0975 -0.0474 0.0317 -0.9936>}
 ball_r {parent=foot_r, loc=<6.1614 -13.1469 -0.6807>, rot=<0.0001 -0.0001 -0.7063 -0.7079>}
 calf_twist_02_r {parent=calf_r, loc=<15.9268 0.0967 -0.0397>, rot=<0.0237 0.0095 -0.0035 -0.9997>}
 calf_twist_01_r {parent=calf_r, loc=<32.9335 0.0967 -0.0397>, rot=<0.0237 0.0095 -0.0035 -0.9997>}
 deform_knee_in_base_r {parent=calf_r, loc=<-1.2981 8.0356 0.0496>, rot=<0 0 0 -1>}
 deform_knee_in_r {parent=deform_knee_in_base_r, loc=<0 0 0>, rot=<0 0 0 1>}
 deform_knee_out_base_r {parent=calf_r, loc=<0 -6.8739 0>, rot=<0 0 0 -1>}
 deform_knee_out_r {parent=deform_knee_out_base_r, loc=<0 0 0>, rot=<0 0 0 1>}
 deform_calf_box_r_MSTR {parent=calf_r, loc=<21.5 0.0967 -0.0397>, rot=<0.0237 0.0095 -0.0035 -0.9997>}
 deform_calf_box_top_r {parent=deform_calf_box_r_MSTR, loc=<-11.0565 0 0>, rot=<0 0 0 1>}
 dyn_calf_box_flap_top_in_r_MSTR {parent=deform_calf_box_top_r, loc=<-0.2138 0.2194 -8.2797>, rot=<-0.2237 0.0461 0.9733 0.0241>}
 dyn_calf_box_flap_top_in_r {parent=dyn_calf_box_flap_top_in_r_MSTR, loc=<0 0 0>, rot=<0 0 0 1>}
 dyn_calf_box_flap_top_r_bk_r_MSTR {parent=deform_calf_box_top_r, loc=<0.5843 7.1912 -0.2511>, rot=<-0.2518 -0.6556 0.661 0.2643>}
 dyn_calf_box_flap_top_r_bk_r {parent=dyn_calf_box_flap_top_r_bk_r_MSTR, loc=<0 0 0>, rot=<0 0 0 1>}
 dyn_calf_box_flap_top_r_out_r_MSTR {parent=deform_calf_box_top_r, loc=<0.3165 -1.1805 7.3488>, rot=<-0.0096 -0.9394 -0.0568 0.338>}
 dyn_calf_box_flap_top_r_out_r {parent=dyn_calf_box_flap_top_r_out_r_MSTR, loc=<0 0 0>, rot=<0 0 0 1>}
 dyn_calf_box_flap_top_fr_r_MSTR {parent=deform_calf_box_top_r, loc=<-0.0495 -8.2358 -0.9642>, rot=<-0.2312 0.5706 0.7593 -0.2108>}
 dyn_calf_box_flap_top_fr_r {parent=dyn_calf_box_flap_top_fr_r_MSTR, loc=<0 0 0>, rot=<0 0 0 1>}
 deform_calf_box_bottom_r {parent=deform_calf_box_r_MSTR, loc=<12.4124 0 0>, rot=<0 0 0 1>}
 dyn_calf_box_flap_bottom_out_r_MSTR {parent=deform_calf_box_bottom_r, loc=<1.1529 -1.8252 7.7322>, rot=<0.0123 0.2937 0.0164 -0.9557>}
 dyn_calf_box_flap_bottom_out_r {parent=dyn_calf_box_flap_bottom_out_r_MSTR, loc=<0 0 0>, rot=<0 0 0 1>}
 dyn_calf_box_flap_bottom_bk_r_MSTR {parent=deform_calf_box_bottom_r, loc=<0.6957 7.7891 -1.1124>, rot=<0.6472 0.2316 -0.3108 -0.6564>}
 dyn_calf_box_flap_bottom_bk_r {parent=dyn_calf_box_flap_bottom_bk_r_MSTR, loc=<0 0 0>, rot=<0 0 0 1>}
 dyn_calf_box_flap_bottom_in_r_MSTR {parent=deform_calf_box_bottom_r, loc=<-0.5771 -0.3782 -9.4658>, rot=<0.9463 -0.0116 -0.3176 -0.0589>}
 dyn_calf_box_flap_bottom_in_r {parent=dyn_calf_box_flap_bottom_in_r_MSTR, loc=<0 0 0>, rot=<0 0 0 1>}
 dyn_calf_box_flap_bottom_fr_r_MSTR {parent=deform_calf_box_bottom_r, loc=<0.3571 -7.8089 -0.4516>, rot=<-0.6443 0.3096 0.4495 -0.5358>}
 dyn_calf_box_flap_bottom_fr_r {parent=dyn_calf_box_flap_bottom_fr_r_MSTR, loc=<0 0 0>, rot=<0 0 0 1>}
 thigh_twist_01_r {parent=thigh_r, loc=<21.6303 0 0>, rot=<0.0013 0 0 -1>}
 deform_groin_base_r {parent=thigh_r, loc=<2.1908 -8.5661 1.7442>, rot=<0 0 0 -1>}
 deform_groin_r {parent=deform_groin_base_r, loc=<0 0 0>, rot=<0 0 0 -1>}
 deform_glute_base_l {parent=pelvis, loc=<-10.1877 -11.6217 -12.899>, rot=<0.0275 0.0501 -0.4472 -0.8926>}
 deform_glute_l {parent=deform_glute_base_l, loc=<0 0 0>, rot=<0 0 0 1>}
 deform_glute_base_r {parent=pelvis, loc=<-10.1877 -11.6218 12.899>, rot=<0.0501 -0.0275 0.8926 -0.4472>}
 deform_glute_r {parent=deform_glute_base_r, loc=<0 0 0>, rot=<0 0 0 1>}
 deform_pelvis_box_MSTR {parent=pelvis, loc=<0 0 0>, rot=<0 0 0 1>}
 deform_pelvis_box_front {parent=deform_pelvis_box_MSTR, loc=<0 17.4642 0>, rot=<0.5 0.5 0.5 0.5>}
 dyn_flap_front_MSTR {parent=deform_pelvis_box_front, loc=<0.8418 -0.1927 -10.1469>, rot=<0.0014 -0.5005 -0.0023 -0.8658>}
 dyn_flap_front {parent=dyn_flap_front_MSTR, loc=<0 0 0>, rot=<0 0 0 1>}
 deform_pelvis_box_r {parent=deform_pelvis_box_MSTR, loc=<0.5494 3.0797 17.7323>, rot=<0.707 -0.0127 0.707 0.0127>}
 dyn_flap_r_MSTR {parent=deform_pelvis_box_r, loc=<-0.0136 -0.1718 -10.5565>, rot=<0 -0.4346 0 -0.9006>}
 dyn_flap_r {parent=dyn_flap_r_MSTR, loc=<0 0 0>, rot=<0 0 0 1>}
 deform_switch_MSTR {parent=deform_pelvis_box_r, loc=<1.3034 1.0851 2.4938>, rot=<0 0 0 1>}
 deform_switch {parent=deform_switch_MSTR, loc=<0 0 0>, rot=<0 0 0 1>}
 deform_pelvis_box_l {parent=deform_pelvis_box_MSTR, loc=<0.882 3.7801 -17.5478>, rot=<-0.0127 -0.707 0.0127 -0.707>}
 dyn_flap_l_MSTR {parent=deform_pelvis_box_l, loc=<0.0034 -0.8247 -11.4087>, rot=<0 -0.4292 0 -0.9032>}
 dyn_flap_l {parent=dyn_flap_l_MSTR, loc=<0 0 0>, rot=<0 0 0 1>}
 deform_squeeze_tubes_MSTR {parent=deform_pelvis_box_l, loc=<3.4959 -5.4323 5.3216>, rot=<0 0 0 1>}
 deform_squeeze_tubes {parent=deform_squeeze_tubes_MSTR, loc=<0 0 0>, rot=<0 0 0 1>}
 deform_pelvis_box_back {parent=deform_pelvis_box_MSTR, loc=<-0.8371 -13.4104 0.5293>, rot=<0.491 -0.5089 0.5089 -0.491>}
 dyn_flap_back_MSTR {parent=deform_pelvis_box_back, loc=<0.2773 0.9171 -10.0893>, rot=<0 -0.4839 0 -0.8751>}
 dyn_flap_back {parent=dyn_flap_back_MSTR, loc=<0 0 0>, rot=<0 0 0 1>}
 deform_power_strip_MSTR {parent=deform_pelvis_box_back, loc=<3.074 1.247 5.7197>, rot=<0 0 0 1>}
 deform_power_strip {parent=deform_power_strip_MSTR, loc=<0 0 0>, rot=<0 0 0 1>}
 upperarm_l_up_SPR {type=Curve, loc=<0 0 0>, rot=<0 0 0 1>}
 upperarm_r_up_SPR {type=Curve, loc=<0 0 0>, rot=<0 0 0 1>}
]
variables [
 FEAdjustAlphaLeft {type=double, default=0}
 FEAdjustAlphaRight {type=double, default=0}
 crouchMoveAlpha {type=double, default=0}
 isCrouchMoving {type=bool, default=False}
 isCustomIdlePlaying {type=bool, default=False}
]
