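# Physics asset for the large male MorningSoak Cook body cosmetic, covering four dynamic flaps (front, back, left, right) and both thighs.
# Each flap has a long capsule constrained to a tiny capsule on its MSTR bone.
# The constraints lock one swing axis, limit the other and allow slight twist, so each flap moves mostly like a hinge.
# Unconstrained capsules on the thighs provide leg shapes alongside the flaps.
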PhysicsAsset /BRCosmetics/Characters/Player/Male/Large/Bodies/M_LRG_MorningSoak_Cook/Meshes/M_LRG_MorningSoak_Cook_Physics
bodies {
 dyn_flap_back [{shapeType=Capsule, length=44.7521, radius=6.1961, shapeName=dyn_flap_back_capsule, center=<7.936 0 3.6>, rotation=<rot 0 0.59 -89.3064>}]
 dyn_flap_back_MSTR [{shapeType=Capsule, length=0.505, radius=0.505, shapeName=dyn_flap_back_MSTR_capsule, rotation=<rot 0 -90 0>}]
 dyn_flap_front [{
   shapeType=Capsule
   length=44.7521
   radius=6.1961
   shapeName=dyn_flap_front_capsule
   center=<7.2984 0 3.5>
   rotation=<rot 0 -179.3139 -88.6933>
  }]
 dyn_flap_front_MSTR [{shapeType=Capsule, length=0.505, radius=0.505, shapeName=dyn_flap_front_MSTR_capsule, rotation=<rot 0 -90 0>}]
 dyn_flap_l [{shapeType=Capsule, length=43.1084, radius=5.7926, shapeName=dyn_flap_l_capsule, center=<8.3426 0 3.4>, rotation=<rot 0 178.3554 -88.8409>}]
 dyn_flap_l_MSTR [{shapeType=Capsule, length=0.505, radius=0.505, shapeName=dyn_flap_l_MSTR_capsule, rotation=<rot 0 -90 0>}]
 dyn_flap_r [{shapeType=Capsule, length=42.7049, radius=6.1961, shapeName=dyn_flap_r_capsule, center=<8.2576 0 3.6>, rotation=<rot 0 178.342 -88.3945>}]
 dyn_flap_r_MSTR [{shapeType=Capsule, length=0.505, radius=0.505, shapeName=dyn_flap_r_MSTR_capsule, rotation=<rot 0 -90 0>}]
 thigh_l [{shapeType=Capsule, length=47.155, radius=7.8135, shapeName=thigh_l_capsule, center=<-26.9167 1.4471 1.8487>, rotation=<rot 90 0 0>}]
 thigh_r [{shapeType=Capsule, length=47.155, radius=7.8135, shapeName=thigh_r_capsule, center=<26.2439 -1.4471 -1.8487>, rotation=<rot 90 0 0>}]
}
constraints [
 {bone1Name=dyn_flap_back, bone2Name=dyn_flap_back_MSTR, swing1LimitDegrees=0, swing2LimitDegrees=40, twistLimitDegrees=2}
 {bone1Name=dyn_flap_front, bone2Name=dyn_flap_front_MSTR, swing1LimitDegrees=0, swing2LimitDegrees=40, twistLimitDegrees=2}
 {bone1Name=dyn_flap_l, bone2Name=dyn_flap_l_MSTR, swing1LimitDegrees=0, swing2LimitDegrees=40, twistLimitDegrees=2}
 {bone1Name=dyn_flap_r, bone2Name=dyn_flap_r_MSTR, swing1LimitDegrees=0, swing2LimitDegrees=40, twistLimitDegrees=2}
]
