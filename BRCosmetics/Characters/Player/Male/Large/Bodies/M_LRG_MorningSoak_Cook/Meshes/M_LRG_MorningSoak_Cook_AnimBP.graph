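# Post-process-style AnimBP for the MorningSoak Cook large male body, adding secondary motion to its box costume.
# Copies the parent mesh pose into a Control Rig fed crouch-move and custom-idle flags, then chains pelvis and right-arm RigidBody sims with spring-driven AnimDynamics for boxes and flaps.
# Events set dynamics alphas for skydiving, rails, ziplines and emotes; updates overwrite AnimDynamicsAlpha.
# The shown AnimGraph never reads those alphas; AnimDynamics nodes show Alpha 1.
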AnimBlueprint /BRCosmetics/Characters/Player/Male/Large/Bodies/M_LRG_MorningSoak_Cook/Meshes/M_LRG_MorningSoak_Cook_AnimBP
target_skeleton=&/M_LRG_MorningSoak_Cook_Skeleton
graphs [
 AnimGraph {nodes [
   n1 = Control Rig!(
    Source=Copy Pose From Mesh() {bCopyCurves=true, bUseAttachedParent=true}
    Alpha
    isCrouchMoving:bool=Get isCrouchMoving() -> (isCrouchMoving:bool)
    isCustomIdlePlaying:bool=Get Custom Idle Is Playing() -> (Custom Idle Is Playing:bool)
   ) -> (Pose) {controlRigAssetReference {blueprintRigClass=&/M_LRG_MorningSoak_Cook_CtrlRig.M_LRG_MorningSoak_Cook_CtrlRig_C}}
   n2 = RigidBody(ComponentPose=Local To Component(LocalPose=@n1)) {
    bOverrideWorldGravity=true
    overridePhysicsAsset=&/M_LRG_MorningSoak_Cook_Physics
    simulationSpace=BaseBoneSpace
    baseBoneRef {boneName=pelvis}
    overrideWorldGravity {z=-980}
    simSpaceSettings {
     dampingAlpha=0
     maxAngularAcceleration=6000
     maxAngularVelocity=12
     maxLinearAcceleration=4000
     maxLinearVelocity=800
     worldAlpha=1
     externalLinearDragV=<0.1 0.1 0.1>
    }
   }
   n3 = RigidBody(ComponentPose=@n2) {
    bOverrideWorldGravity=true
    overridePhysicsAsset=&/M_LRG_MorningSoak_Cook_Arm_R_Physics
    simulationSpace=BaseBoneSpace
    baseBoneRef {boneName=upperarm_r}
    overrideWorldGravity {z=-980}
    simSpaceSettings {
     dampingAlpha=0
     maxAngularAcceleration=6000
     maxAngularVelocity=12
     maxLinearAcceleration=3000
     maxLinearVelocity=600
     worldAlpha=1
     externalLinearDragV=<0.1 0.1 0.1>
    }
   }
   n4 = Anim Dynamics - Bone: deform_pelvis_box_MSTR(ComponentPose=@n3, Alpha=1, ExternalForce="0,0,0") {
    angularDampingOverride=0.999
    angularSpringConstant=1600
    bAngularSpring=true
    bOverrideAngularDamping=true
    bUseGravityOverride=true
    simulationSpace=BoneRelative
    boundBone {boneName=deform_pelvis_box_MSTR}
    componentAppliedLinearAccClamp=<1500 1500 1500>
    gravityOverride {z=-980}
    physicsBodyDefinitions [{
      boundBone {boneName=deform_pelvis_box_MSTR}
      boxExtents=<60 60 60>
      constraintSetup {angularLimitsMax=<1 1 1>, angularLimitsMin=<-1 -1 -1>, angularTarget=<1 0 0>}
      localJointOffset=<1 0 0>
     }]
    relativeSpaceBone {boneName=pelvis}
    simSpaceSettings {maxAngularAcceleration=40, maxAngularVelocity=8, simSpaceAngularAlpha=0.25}
   }
   n5 = Anim Dynamics - Bone: deform_calf_box_l_MSTR(ComponentPose=@n4, Alpha=1, ExternalForce="0,0,0") {
    angularDampingOverride=0.9
    angularSpringConstant=1600
    bAngularSpring=true
    bOverrideAngularDamping=true
    bUseGravityOverride=true
    simulationSpace=BoneRelative
    boundBone {boneName=deform_calf_box_l_MSTR}
    componentAppliedLinearAccClamp=<1500 1500 1500>
    gravityOverride {z=-980}
    physicsBodyDefinitions [{
      boundBone {boneName=deform_calf_box_l_MSTR}
      boxExtents=<60 60 60>
      constraintSetup {angularLimitsMax=<3 2 2>, angularLimitsMin=<-3 -2 -2>, angularTarget=<1 0 0>}
      localJointOffset=<5 0 0>
     }]
    relativeSpaceBone {boneName=calf_l}
    simSpaceSettings {maxAngularAcceleration=40, maxAngularVelocity=8, simSpaceAngularAlpha=0.2}
   }
   n6 = Anim Dynamics - Bone: deform_calf_box_r_MSTR(ComponentPose=@n5, Alpha=1, ExternalForce="0,0,0") {
    angularDampingOverride=0.9
    angularSpringConstant=1600
    bAngularSpring=true
    bOverrideAngularDamping=true
    bUseGravityOverride=true
    simulationSpace=BoneRelative
    boundBone {boneName=deform_calf_box_r_MSTR}
    componentAppliedLinearAccClamp=<1500 1500 1500>
    gravityOverride {z=-980}
    physicsBodyDefinitions [{
      boundBone {boneName=deform_calf_box_r_MSTR}
      boxExtents=<60 60 60>
      constraintSetup {angularLimitsMax=<3 2 2>, angularLimitsMin=<-3 -2 -2>, angularTarget=<1 0 0>}
      localJointOffset=<7 0 0>
     }]
    relativeSpaceBone {boneName=calf_r}
    simSpaceSettings {maxAngularAcceleration=40, maxAngularVelocity=8, simSpaceAngularAlpha=0.2}
   }
   n7 = Anim Dynamics - Bone: deform_lowerarm_box_l_MSTR(ComponentPose=@n6, Alpha=1, ExternalForce="0,0,0") {
    angularDampingOverride=0.9
    angularSpringConstant=1600
    bAngularSpring=true
    bOverrideAngularDamping=true
    bUseGravityOverride=true
    simulationSpace=RootRelative
    boundBone {boneName=deform_lowerarm_box_l_MSTR}
    componentAppliedLinearAccClamp=<1500 1500 1500>
    gravityOverride {z=-980}
    physicsBodyDefinitions [{
      boundBone {boneName=deform_lowerarm_box_l_MSTR}
      boxExtents=<60 60 60>
      constraintSetup {angularLimitsMax=<1 1 1>, angularLimitsMin=<-1 -1 -1>, angularTarget=<1 0 0>}
      localJointOffset=<6 0 0>
     }]
    relativeSpaceBone {boneName=lowerarm_l}
    simSpaceSettings {maxAngularAcceleration=40, maxAngularVelocity=8, simSpaceAngularAlpha=0.4}
   }
   n8 = Anim Dynamics - Bone: deform_lowerarm_box_r_MSTR(ComponentPose=@n7, Alpha=1, ExternalForce="0,0,0") {
    angularDampingOverride=0.9
    angularSpringConstant=1700
    bAngularSpring=true
    bOverrideAngularDamping=true
    bUseGravityOverride=true
    simulationSpace=RootRelative
    boundBone {boneName=deform_lowerarm_box_r_MSTR}
    componentAppliedLinearAccClamp=<1500 1500 1500>
    gravityOverride {z=-980}
    physicsBodyDefinitions [{
      boundBone {boneName=deform_lowerarm_box_r_MSTR}
      boxExtents=<60 60 60>
      constraintSetup {angularLimitsMax=<1 1 1>, angularLimitsMin=<-1 -1 -1>, angularTarget=<1 0 0>}
      localJointOffset=<-6 0 0>
     }]
    relativeSpaceBone {boneName=lowerarm_r}
    simSpaceSettings {maxAngularAcceleration=40, maxAngularVelocity=8, simSpaceAngularAlpha=0.4}
   }
   n9 = Anim Dynamics - Bone: dyn_lowerarm_flap_fr_l(ComponentPose=@n8, Alpha=1, ExternalForce="0,0,0") {
    angularDampingOverride=0.9
    angularSpringConstant=1600
    bAngularSpring=true
    bOverrideAngularDamping=true
    bUseGravityOverride=true
    simulationSpace=RootRelative
    boundBone {boneName=dyn_lowerarm_flap_fr_l}
    componentAppliedLinearAccClamp=<1500 1500 2>
    gravityOverride {z=-980}
    physicsBodyDefinitions [{
      boundBone {boneName=dyn_lowerarm_flap_fr_l}
      boxExtents=<30 30 30>
      constraintSetup {angularLimitsMax=<0 7.5 0>, angularLimitsMin=<0 -20 0>, angularTarget=<1 0 0>}
      localJointOffset=<3 0 0>
     }]
    relativeSpaceBone {boneName=lowerarm_l}
    simSpaceSettings {maxAngularAcceleration=40, maxAngularVelocity=8, simSpaceAngularAlpha=0.4}
   }
   n10 = Anim Dynamics - Bone: dyn_lowerarm_flap_out_l(ComponentPose=@n9, Alpha=1, ExternalForce="0,0,0") {
    angularDampingOverride=0.9
    angularSpringConstant=1600
    bAngularSpring=true
    bOverrideAngularDamping=true
    bUseGravityOverride=true
    simulationSpace=RootRelative
    boundBone {boneName=dyn_lowerarm_flap_out_l}
    componentAppliedLinearAccClamp=<1500 1500 2>
    gravityOverride {z=-980}
    physicsBodyDefinitions [{
      boundBone {boneName=dyn_lowerarm_flap_out_l}
      boxExtents=<30 30 30>
      constraintSetup {angularLimitsMax=<0 7.5 0>, angularLimitsMin=<0 -20 0>, angularTarget=<1 0 0>}
      localJointOffset=<5 0 0>
     }]
    relativeSpaceBone {boneName=lowerarm_l}
    simSpaceSettings {maxAngularAcceleration=40, maxAngularVelocity=8, simSpaceAngularAlpha=0.4}
   }
   n11 = Anim Dynamics - Bone: dyn_lowerarm_flap_back_l(ComponentPose=@n10, Alpha=1, ExternalForce="0,0,0") {
    angularDampingOverride=0.9
    angularSpringConstant=1600
    bAngularSpring=true
    bOverrideAngularDamping=true
    bUseGravityOverride=true
    simulationSpace=RootRelative
    boundBone {boneName=dyn_lowerarm_flap_back_l}
    componentAppliedLinearAccClamp=<1500 1500 2>
    gravityOverride {z=-980}
    physicsBodyDefinitions [{
      boundBone {boneName=dyn_lowerarm_flap_back_l}
      boxExtents=<30 30 30>
      constraintSetup {angularLimitsMax=<0 7.5 0>, angularLimitsMin=<0 -20 0>, angularTarget=<1 0 0>}
      localJointOffset=<7 0 0>
     }]
    relativeSpaceBone {boneName=lowerarm_l}
    simSpaceSettings {maxAngularAcceleration=40, maxAngularVelocity=8, simSpaceAngularAlpha=0.4}
   }
   n12 = Anim Dynamics - Bone: dyn_lowerarm_flap_in_l(ComponentPose=@n11, Alpha=1, ExternalForce="0,0,0") {
    angularDampingOverride=0.9
    angularSpringConstant=1600
    bAngularSpring=true
    bOverrideAngularDamping=true
    bUseGravityOverride=true
    simulationSpace=RootRelative
    boundBone {boneName=dyn_lowerarm_flap_in_l}
    componentAppliedLinearAccClamp=<1500 1500 2>
    gravityOverride {z=-980}
    physicsBodyDefinitions [{
      boundBone {boneName=dyn_lowerarm_flap_in_l}
      boxExtents=<30 30 30>
      constraintSetup {angularLimitsMax=<0 7.5 0>, angularLimitsMin=<0 -20 0>, angularTarget=<1 0 0>}
      localJointOffset=<4 0 0>
     }]
    relativeSpaceBone {boneName=lowerarm_l}
    simSpaceSettings {maxAngularAcceleration=40, maxAngularVelocity=8, simSpaceAngularAlpha=0.4}
   }
   n13 = Anim Dynamics - Bone: dyn_lowerarm_flap_fr_r(ComponentPose=@n12, Alpha=1, ExternalForce="0,0,0") {
    angularDampingOverride=0.9
    angularSpringConstant=1600
    bAngularSpring=true
    bOverrideAngularDamping=true
    bUseGravityOverride=true
    simulationSpace=RootRelative
    boundBone {boneName=dyn_lowerarm_flap_fr_r}
    componentAppliedLinearAccClamp=<1500 1500 2>
    gravityOverride {z=-980}
    physicsBodyDefinitions [{
      boundBone {boneName=dyn_lowerarm_flap_fr_r}
      boxExtents=<30 30 30>
      constraintSetup {angularLimitsMax=<0 7.5 0>, angularLimitsMin=<0 -20 0>, angularTarget=<1 0 0>}
      localJointOffset=<6 0 0>
     }]
    relativeSpaceBone {boneName=lowerarm_r}
    simSpaceSettings {maxAngularAcceleration=40, maxAngularVelocity=8, simSpaceAngularAlpha=0.4}
   }
   n14 = Anim Dynamics - Bone: dyn_lowerarm_flap_out_r(ComponentPose=@n13, Alpha=1, ExternalForce="0,0,0") {
    angularDampingOverride=0.9
    angularSpringConstant=1600
    bAngularSpring=true
    bOverrideAngularDamping=true
    bUseGravityOverride=true
    simulationSpace=RootRelative
    boundBone {boneName=dyn_lowerarm_flap_out_r}
    componentAppliedLinearAccClamp=<1500 1500 2>
    gravityOverride {z=-980}
    physicsBodyDefinitions [{
      boundBone {boneName=dyn_lowerarm_flap_out_r}
      boxExtents=<30 30 30>
      constraintSetup {angularLimitsMax=<0 7.5 0>, angularLimitsMin=<0 -20 0>, angularTarget=<1 0 0>}
      localJointOffset=<5 0 0>
     }]
    relativeSpaceBone {boneName=lowerarm_r}
    simSpaceSettings {maxAngularAcceleration=40, maxAngularVelocity=8, simSpaceAngularAlpha=0.4}
   }
   n15 = Anim Dynamics - Bone: dyn_lowerarm_flap_back_r(ComponentPose=@n14, Alpha=1, ExternalForce="0,0,0") {
    angularDampingOverride=0.9
    angularSpringConstant=1600
    bAngularSpring=true
    bOverrideAngularDamping=true
    bUseGravityOverride=true
    simulationSpace=RootRelative
    boundBone {boneName=dyn_lowerarm_flap_back_r}
    componentAppliedLinearAccClamp=<1500 1500 2>
    gravityOverride {z=-980}
    physicsBodyDefinitions [{
      boundBone {boneName=dyn_lowerarm_flap_back_r}
      boxExtents=<30 30 30>
      constraintSetup {angularLimitsMax=<0 7.5 0>, angularLimitsMin=<0 -20 0>, angularTarget=<1 0 0>}
      localJointOffset=<3 0 0>
     }]
    relativeSpaceBone {boneName=lowerarm_r}
    simSpaceSettings {maxAngularAcceleration=40, maxAngularVelocity=8, simSpaceAngularAlpha=0.4}
   }
   n16 = Anim Dynamics - Bone: dyn_lowerarm_flap_in_r(ComponentPose=@n15, Alpha=1, ExternalForce="0,0,0") {
    angularDampingOverride=0.9
    angularSpringConstant=1600
    bAngularSpring=true
    bOverrideAngularDamping=true
    bUseGravityOverride=true
    simulationSpace=RootRelative
    boundBone {boneName=dyn_lowerarm_flap_in_r}
    componentAppliedLinearAccClamp=<1500 1500 2>
    gravityOverride {z=-980}
    physicsBodyDefinitions [{
      boundBone {boneName=dyn_lowerarm_flap_in_r}
      boxExtents=<30 30 30>
      constraintSetup {angularLimitsMax=<0 7.5 0>, angularLimitsMin=<0 -20 0>, angularTarget=<1 0 0>}
      localJointOffset=<7 0 0>
     }]
    relativeSpaceBone {boneName=lowerarm_r}
    simSpaceSettings {maxAngularAcceleration=40, maxAngularVelocity=8, simSpaceAngularAlpha=0.4}
   }
   n17 = Anim Dynamics - Bone: dyn_calf_box_flap_top_fr_l(ComponentPose=@n16, Alpha=1, ExternalForce="0,0,0") {
    angularDampingOverride=0.9
    angularSpringConstant=1600
    bAngularSpring=true
    bOverrideAngularDamping=true
    bUseGravityOverride=true
    simulationSpace=RootRelative
    boundBone {boneName=dyn_calf_box_flap_top_fr_l}
    componentAppliedLinearAccClamp=<1500 1500 1500>
    gravityOverride {z=-980}
    physicsBodyDefinitions [{
      boundBone {boneName=dyn_calf_box_flap_top_fr_l}
      boxExtents=<20 20 20>
      constraintSetup {angularLimitsMax=<0 30 0>, angularLimitsMin=<0 -20 0>, angularTarget=<1 0 0>}
      localJointOffset=<3 0 0>
     }]
    relativeSpaceBone {boneName=calf_twist_01_l}
    simSpaceSettings {maxAngularAcceleration=40, maxAngularVelocity=8, simSpaceAngularAlpha=0.6}
   }
   n18 = Anim Dynamics - Bone: dyn_calf_box_flap_top_out_l(ComponentPose=@n17, Alpha=1, ExternalForce="0,0,0") {
    angularDampingOverride=0.9
    angularSpringConstant=1600
    bAngularSpring=true
    bOverrideAngularDamping=true
    bUseGravityOverride=true
    simulationSpace=RootRelative
    boundBone {boneName=dyn_calf_box_flap_top_out_l}
    componentAppliedLinearAccClamp=<1500 1500 1500>
    gravityOverride {z=-980}
    physicsBodyDefinitions [{
      boundBone {boneName=dyn_calf_box_flap_top_out_l}
      boxExtents=<20 20 20>
      constraintSetup {angularLimitsMax=<0 30 0>, angularLimitsMin=<0 -20 0>, angularTarget=<1 0 0>}
      localJointOffset=<5 0 0>
     }]
    relativeSpaceBone {boneName=calf_twist_01_l}
    simSpaceSettings {maxAngularAcceleration=40, maxAngularVelocity=8, simSpaceAngularAlpha=0.6}
   }
   n19 = Anim Dynamics - Bone: dyn_calf_box_flap_top_bk_l(ComponentPose=@n18, Alpha=1, ExternalForce="0,0,0") {
    angularDampingOverride=0.9
    angularSpringConstant=1600
    bAngularSpring=true
    bOverrideAngularDamping=true
    bUseGravityOverride=true
    simulationSpace=RootRelative
    boundBone {boneName=dyn_calf_box_flap_top_bk_l}
    componentAppliedLinearAccClamp=<1500 1500 1500>
    gravityOverride {z=-980}
    physicsBodyDefinitions [{
      boundBone {boneName=dyn_calf_box_flap_top_bk_l}
      boxExtents=<20 20 20>
      constraintSetup {angularLimitsMax=<0 30 0>, angularLimitsMin=<0 -20 0>, angularTarget=<1 0 0>}
      localJointOffset=<7 0 0>
     }]
    relativeSpaceBone {boneName=calf_twist_01_l}
    simSpaceSettings {maxAngularAcceleration=40, maxAngularVelocity=8, simSpaceAngularAlpha=0.6}
   }
   n20 = Anim Dynamics - Bone: dyn_calf_box_flap_top_in_l(ComponentPose=@n19, Alpha=1, ExternalForce="0,0,0") {
    angularDampingOverride=0.9
    angularSpringConstant=1600
    bAngularSpring=true
    bOverrideAngularDamping=true
    bUseGravityOverride=true
    simulationSpace=RootRelative
    boundBone {boneName=dyn_calf_box_flap_top_in_l}
    componentAppliedLinearAccClamp=<1500 1500 1500>
    gravityOverride {z=-980}
    physicsBodyDefinitions [{
      boundBone {boneName=dyn_calf_box_flap_top_in_l}
      boxExtents=<20 20 20>
      constraintSetup {angularLimitsMax=<0 30 0>, angularLimitsMin=<0 -20 0>, angularTarget=<1 0 0>}
      localJointOffset=<4 0 0>
     }]
    relativeSpaceBone {boneName=calf_twist_01_l}
    simSpaceSettings {maxAngularAcceleration=40, maxAngularVelocity=8, simSpaceAngularAlpha=0.6}
   }
   n21 = Anim Dynamics - Bone: dyn_calf_box_flap_bottom_fr_l(ComponentPose=@n20, Alpha=1, ExternalForce="0,0,0") {
    angularDampingOverride=0.9
    angularSpringConstant=1600
    bAngularSpring=true
    bOverrideAngularDamping=true
    bUseGravityOverride=true
    simulationSpace=RootRelative
    boundBone {boneName=dyn_calf_box_flap_bottom_fr_l}
    componentAppliedLinearAccClamp=<1500 1500 1500>
    gravityOverride {z=-980}
    physicsBodyDefinitions [{
      boundBone {boneName=dyn_calf_box_flap_bottom_fr_l}
      boxExtents=<20 20 20>
      constraintSetup {angularLimitsMax=<0 0 0>, angularLimitsMin=<0 -20 0>, angularTarget=<1 0 0>}
      localJointOffset=<6 0 0>
     }]
    relativeSpaceBone {boneName=calf_twist_01_l}
    simSpaceSettings {maxAngularAcceleration=40, maxAngularVelocity=8, simSpaceAngularAlpha=0.6}
   }
   n22 = Anim Dynamics - Bone: dyn_calf_box_flap_bottom_out_l(ComponentPose=@n21, Alpha=1, ExternalForce="0,0,0") {
    angularDampingOverride=0.9
    angularSpringConstant=1600
    bAngularSpring=true
    bOverrideAngularDamping=true
    bUseGravityOverride=true
    simulationSpace=RootRelative
    boundBone {boneName=dyn_calf_box_flap_bottom_out_l}
    componentAppliedLinearAccClamp=<1500 1500 1500>
    gravityOverride {z=-980}
    physicsBodyDefinitions [{
      boundBone {boneName=dyn_calf_box_flap_bottom_out_l}
      boxExtents=<20 20 20>
      constraintSetup {angularLimitsMax=<0 30 0>, angularLimitsMin=<0 -20 0>, angularTarget=<1 0 0>}
      localJointOffset=<5 0 0>
     }]
    relativeSpaceBone {boneName=calf_twist_01_l}
    simSpaceSettings {maxAngularAcceleration=40, maxAngularVelocity=8, simSpaceAngularAlpha=0.6}
   }
   n23 = Anim Dynamics - Bone: dyn_calf_box_flap_bottom_bk_l(ComponentPose=@n22, Alpha=1, ExternalForce="0,0,0") {
    angularDampingOverride=0.9
    angularSpringConstant=1600
    bAngularSpring=true
    bOverrideAngularDamping=true
    bUseGravityOverride=true
    simulationSpace=RootRelative
    boundBone {boneName=dyn_calf_box_flap_bottom_bk_l}
    componentAppliedLinearAccClamp=<1500 1500 1500>
    gravityOverride {z=-980}
    physicsBodyDefinitions [{
      boundBone {boneName=dyn_calf_box_flap_bottom_bk_l}
      boxExtents=<20 20 20>
      constraintSetup {angularLimitsMax=<0 30 0>, angularLimitsMin=<0 -20 0>, angularTarget=<1 0 0>}
      localJointOffset=<3 0 0>
     }]
    relativeSpaceBone {boneName=calf_twist_01_l}
    simSpaceSettings {maxAngularAcceleration=40, maxAngularVelocity=8, simSpaceAngularAlpha=0.6}
   }
   n24 = Anim Dynamics - Bone: dyn_calf_box_flap_bottom_in_l(ComponentPose=@n23, Alpha=1, ExternalForce="0,0,0") {
    angularDampingOverride=0.9
    angularSpringConstant=1600
    bAngularSpring=true
    bOverrideAngularDamping=true
    bUseGravityOverride=true
    simulationSpace=RootRelative
    boundBone {boneName=dyn_calf_box_flap_bottom_in_l}
    componentAppliedLinearAccClamp=<1500 1500 1500>
    gravityOverride {z=-980}
    physicsBodyDefinitions [{
      boundBone {boneName=dyn_calf_box_flap_bottom_in_l}
      boxExtents=<20 20 20>
      constraintSetup {angularLimitsMax=<0 30 0>, angularLimitsMin=<0 -20 0>, angularTarget=<1 0 0>}
      localJointOffset=<7 0 0>
     }]
    relativeSpaceBone {boneName=calf_twist_01_l}
    simSpaceSettings {maxAngularAcceleration=40, maxAngularVelocity=8, simSpaceAngularAlpha=0.6}
   }
   n25 = Anim Dynamics - Bone: dyn_calf_box_flap_top_fr_r(ComponentPose=@n24, Alpha=1, ExternalForce="0,0,0") {
    angularDampingOverride=0.9
    angularSpringConstant=1600
    bAngularSpring=true
    bOverrideAngularDamping=true
    bUseGravityOverride=true
    simulationSpace=RootRelative
    boundBone {boneName=dyn_calf_box_flap_top_fr_r}
    componentAppliedLinearAccClamp=<1500 1500 1500>
    gravityOverride {z=-980}
    physicsBodyDefinitions [{
      boundBone {boneName=dyn_calf_box_flap_top_fr_r}
      boxExtents=<20 20 20>
      constraintSetup {angularLimitsMax=<0 30 0>, angularLimitsMin=<0 -20 0>, angularTarget=<1 0 0>}
      localJointOffset=<3 0 0>
     }]
    relativeSpaceBone {boneName=calf_twist_01_r}
    simSpaceSettings {maxAngularAcceleration=40, maxAngularVelocity=8, simSpaceAngularAlpha=0.6}
   }
   n26 = Anim Dynamics - Bone: dyn_calf_box_flap_top_r_out_r(ComponentPose=@n25, Alpha=1, ExternalForce="0,0,0") {
    angularDampingOverride=0.9
    angularSpringConstant=1600
    bAngularSpring=true
    bOverrideAngularDamping=true
    bUseGravityOverride=true
    simulationSpace=RootRelative
    boundBone {boneName=dyn_calf_box_flap_top_r_out_r}
    componentAppliedLinearAccClamp=<1500 1500 1500>
    gravityOverride {z=-980}
    physicsBodyDefinitions [{
      boundBone {boneName=dyn_calf_box_flap_top_r_out_r}
      boxExtents=<20 20 20>
      constraintSetup {angularLimitsMax=<0 30 0>, angularLimitsMin=<0 -20 0>, angularTarget=<1 0 0>}
      localJointOffset=<5 0 0>
     }]
    relativeSpaceBone {boneName=calf_twist_01_r}
    simSpaceSettings {maxAngularAcceleration=40, maxAngularVelocity=8, simSpaceAngularAlpha=0.6}
   }
   n27 = Anim Dynamics - Bone: dyn_calf_box_flap_top_r_bk_r(ComponentPose=@n26, Alpha=1, ExternalForce="0,0,0") {
    angularDampingOverride=0.9
    angularSpringConstant=1600
    bAngularSpring=true
    bOverrideAngularDamping=true
    bUseGravityOverride=true
    simulationSpace=RootRelative
    boundBone {boneName=dyn_calf_box_flap_top_r_bk_r}
    componentAppliedLinearAccClamp=<1500 1500 1500>
    gravityOverride {z=-980}
    physicsBodyDefinitions [{
      boundBone {boneName=dyn_calf_box_flap_top_r_bk_r}
      boxExtents=<20 20 20>
      constraintSetup {angularLimitsMax=<0 30 0>, angularLimitsMin=<0 -20 0>, angularTarget=<1 0 0>}
      localJointOffset=<7 0 0>
     }]
    relativeSpaceBone {boneName=calf_twist_01_r}
    simSpaceSettings {maxAngularAcceleration=40, maxAngularVelocity=8, simSpaceAngularAlpha=0.6}
   }
   n28 = Anim Dynamics - Bone: dyn_calf_box_flap_top_in_r(ComponentPose=@n27, Alpha=1, ExternalForce="0,0,0") {
    angularDampingOverride=0.9
    angularSpringConstant=1600
    bAngularSpring=true
    bOverrideAngularDamping=true
    bUseGravityOverride=true
    simulationSpace=RootRelative
    boundBone {boneName=dyn_calf_box_flap_top_in_r}
    componentAppliedLinearAccClamp=<1500 1500 1500>
    gravityOverride {z=-980}
    physicsBodyDefinitions [{
      boundBone {boneName=dyn_calf_box_flap_top_in_r}
      boxExtents=<20 20 20>
      constraintSetup {angularLimitsMax=<0 30 0>, angularLimitsMin=<0 -20 0>, angularTarget=<1 0 0>}
      localJointOffset=<4 0 0>
     }]
    relativeSpaceBone {boneName=calf_twist_01_r}
    simSpaceSettings {maxAngularAcceleration=40, maxAngularVelocity=8, simSpaceAngularAlpha=0.6}
   }
   n29 = Anim Dynamics - Bone: dyn_calf_box_flap_bottom_fr_r(ComponentPose=@n28, Alpha=1, ExternalForce="0,0,0") {
    angularDampingOverride=0.9
    angularSpringConstant=1600
    bAngularSpring=true
    bOverrideAngularDamping=true
    bUseGravityOverride=true
    simulationSpace=RootRelative
    boundBone {boneName=dyn_calf_box_flap_bottom_fr_r}
    componentAppliedLinearAccClamp=<1500 1500 1500>
    gravityOverride {z=-980}
    physicsBodyDefinitions [{
      boundBone {boneName=dyn_calf_box_flap_bottom_fr_r}
      boxExtents=<20 20 20>
      constraintSetup {angularLimitsMax=<0 0 0>, angularLimitsMin=<0 -20 0>, angularTarget=<1 0 0>}
      localJointOffset=<6 0 0>
     }]
    relativeSpaceBone {boneName=calf_twist_01_r}
    simSpaceSettings {maxAngularAcceleration=40, maxAngularVelocity=8, simSpaceAngularAlpha=0.6}
   }
   n30 = Anim Dynamics - Bone: dyn_calf_box_flap_bottom_out_r(ComponentPose=@n29, Alpha=1, ExternalForce="0,0,0") {
    angularDampingOverride=0.9
    angularSpringConstant=1600
    bAngularSpring=true
    bOverrideAngularDamping=true
    bUseGravityOverride=true
    simulationSpace=RootRelative
    boundBone {boneName=dyn_calf_box_flap_bottom_out_r}
    componentAppliedLinearAccClamp=<1500 1500 1500>
    gravityOverride {z=-980}
    physicsBodyDefinitions [{
      boundBone {boneName=dyn_calf_box_flap_bottom_out_r}
      boxExtents=<20 20 20>
      constraintSetup {angularLimitsMax=<0 30 0>, angularLimitsMin=<0 -20 0>, angularTarget=<1 0 0>}
      localJointOffset=<5 0 0>
     }]
    relativeSpaceBone {boneName=calf_twist_01_r}
    simSpaceSettings {maxAngularAcceleration=40, maxAngularVelocity=8, simSpaceAngularAlpha=0.6}
   }
   n31 = Anim Dynamics - Bone: dyn_calf_box_flap_bottom_bk_r(ComponentPose=@n30, Alpha=1, ExternalForce="0,0,0") {
    angularDampingOverride=0.9
    angularSpringConstant=1600
    bAngularSpring=true
    bOverrideAngularDamping=true
    bUseGravityOverride=true
    simulationSpace=RootRelative
    boundBone {boneName=dyn_calf_box_flap_bottom_bk_r}
    componentAppliedLinearAccClamp=<1500 1500 1500>
    gravityOverride {z=-980}
    physicsBodyDefinitions [{
      boundBone {boneName=dyn_calf_box_flap_bottom_bk_r}
      boxExtents=<20 20 20>
      constraintSetup {angularLimitsMax=<0 30 0>, angularLimitsMin=<0 -20 0>, angularTarget=<1 0 0>}
      localJointOffset=<3 0 0>
     }]
    relativeSpaceBone {boneName=calf_twist_01_r}
    simSpaceSettings {maxAngularAcceleration=40, maxAngularVelocity=8, simSpaceAngularAlpha=0.6}
   }
   n32 = Anim Dynamics - Bone: dyn_calf_box_flap_bottom_in_r(ComponentPose=@n31, Alpha=1, ExternalForce="0,0,0") {
    angularDampingOverride=0.9
    angularSpringConstant=1600
    bAngularSpring=true
    bOverrideAngularDamping=true
    bUseGravityOverride=true
    simulationSpace=RootRelative
    boundBone {boneName=dyn_calf_box_flap_bottom_in_r}
    componentAppliedLinearAccClamp=<1500 1500 1500>
    gravityOverride {z=-980}
    physicsBodyDefinitions [{
      boundBone {boneName=dyn_calf_box_flap_bottom_in_r}
      boxExtents=<20 20 20>
      constraintSetup {angularLimitsMax=<0 30 0>, angularLimitsMin=<0 -20 0>, angularTarget=<1 0 0>}
      localJointOffset=<7 0 0>
     }]
    relativeSpaceBone {boneName=calf_twist_01_r}
    simSpaceSettings {maxAngularAcceleration=40, maxAngularVelocity=8, simSpaceAngularAlpha=0.6}
   }
   Output Pose(Result=Component To Local(ComponentPose=@n32)) {layerGroup=None}
  ]}
 EventGraph {nodes [
   Event OnBeginSkydiving()
   > Set AnimDynamicsAlpha(AnimDynamicsAlpha:double=0) -> (Output_Get:double)
   > Set torsoBoxDynAlpha(torsoBoxDynAlpha:double=0) -> (Output_Get:double)
   Event OnEndSkydiving()
   > Set AnimDynamicsAlpha(AnimDynamicsAlpha=1)
   > Set torsoBoxDynAlpha(torsoBoxDynAlpha=1)
   Event OnEnterGrindRail() {is=title K2Node_Event}
   > Set AnimDynamicsAlpha(AnimDynamicsAlpha=0)
   > Set torsoBoxDynAlpha(torsoBoxDynAlpha=0)
   n1 = Event BlueprintProcessAnimInputEvent()
   > n2 = Branch(Condition=IsNamed(self=@n1.AnimInputEvent, InEventName=aie.Player.Zipline.Entered))
   > Set AnimDynamicsAlpha(AnimDynamicsAlpha=0)
   > Set torsoBoxDynAlpha(torsoBoxDynAlpha=0)
   Branch(execute=@n2.else, Condition=IsNamed(self=@n1.AnimInputEvent, InEventName=aie.Player.Zipline.Exited))
   > Set AnimDynamicsAlpha(AnimDynamicsAlpha=1)
   > Set torsoBoxDynAlpha(torsoBoxDynAlpha=1)
   Event OnExitGrindRail() {is=title K2Node_Event}
   > Set AnimDynamicsAlpha(AnimDynamicsAlpha=1)
   > Set torsoBoxDynAlpha(torsoBoxDynAlpha=1)
   Event BlueprintUpdateAnimation()
   > n3 = Sequence!(execute) -> (then_0, then_1, then_2:Exec)
   > n4 = Branch(Condition=Get bIsInFrontEnd())
   > Set AnimDynamicsAlpha(AnimDynamicsAlpha=0.8)
   Set AnimDynamicsAlpha(execute=@n4.else, AnimDynamicsAlpha=1)
   Branch(execute=@n3.then_1, Condition=Get bIsInFrontEnd())
   > Set Custom Idle Is Playing Float(Custom Idle Is Playing Float:double=GetCurveValue(self=GetOwnerFrontEndAnimInstance(), CurveName=frontend_idle_bypass)) -> (Output_Get:double)
   > n5 = Branch(Condition="float >= float"(A=Get Custom Idle Is Playing Float() -> (Custom Idle Is Playing Float:double), B=1))
   > Set Custom Idle Is Playing(Custom Idle Is Playing:bool=true) -> (Output_Get:bool)
   Set Custom Idle Is Playing(execute=@n5.else, Custom Idle Is Playing=false)
   Branch(execute=@n3.then_2, Condition=Get bIsCrouching())
   > n6 = Branch(Condition=Get bIsMoving2D())
   > Set isCrouchMoving(isCrouchMoving:bool=true) -> (Output_Get:bool)
   > Set crouchMoveAlpha(crouchMoveAlpha:double=1) -> (Output_Get:double)
   Set isCrouchMoving(execute=@n6.else, isCrouchMoving=false)
   > Set crouchMoveAlpha(crouchMoveAlpha=0)
   Event OnBeginPlayingEmote() {is=title K2Node_Event}
   > Set AnimDynamicsAlpha(AnimDynamicsAlpha=0.25)
   Event OnEndPlayingEmote() {is=title K2Node_Event}
   > Set AnimDynamicsAlpha(AnimDynamicsAlpha=1)
  ]}
]
variables [
 AnimDynamicsAlpha {type=number, default=1}
 crouchMoveAlpha {type=number}
 Custom Idle Is Playing {type=boolean}
 Custom Idle Is Playing Float {type=number}
 isCrouchMoving {type=boolean}
 torsoBoxDynAlpha {type=number, default=1}
]
