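# Physics asset of cape colliders for the large male MorningSoak Cook character body, named for colliding with cape or cloth-like dynamic parts.
# Capsule bodies wrap the pelvis, spine, upper and lower arms, hands, thighs, calves and feet, with two overlapping capsules on the pelvis.
# An additional long, thin capsule sits on the dyn_flap_back bone.
# No constraints are defined, so it holds collision shapes only, not a jointed ragdoll.
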PhysicsAsset /BRCosmetics/Characters/Player/Male/Large/Bodies/M_LRG_MorningSoak_Cook/Meshes/M_LRG_MorningSoak_Cook_CapeColliders
bodies {
 calf_l [{shapeType=Capsule, length=24.354, radius=14.6512, center=<-16.4529 -2.2427 0>, rotation=<rot 90 0 0>}]
 calf_r [{shapeType=Capsule, length=24.354, radius=14.6512, center=<16.4529 2.2427 0>, rotation=<rot 90 0 180>}]
 dyn_flap_back [{
   shapeType=Capsule
   length=52.5784
   radius=7.6013
   shapeName=dyn_flap_back_capsule
   center=<5.8408 -0.9513 -4.868>
   rotation=<rot 0 0.59 -89.3064>
  }]
 foot_l [{shapeType=Capsule, length=18.27, radius=15.9419, center=<-4.6126 6.5844 0>, rotation=<rot 0 14.9998 90>}]
 foot_r [{shapeType=Capsule, length=18.27, radius=15.9419, center=<4.6126 -6.5844 0>, rotation=<rot 0 14.9998 -90>}]
 hand_l [{shapeType=Capsule, length=10.2043, radius=9.8833, center=<6.2703 -0.7956 0>, rotation=<rot 90 0 0>}]
 hand_r [{shapeType=Capsule, length=9.0694, radius=9.032, center=<-8.0165 0.5179 0>, rotation=<rot 90 163.4802 0>}]
 lowerarm_l [{shapeType=Capsule, length=14.2683, radius=14.0337, center=<8.9852 0 1.9278>, rotation=<rot 90 0 0>}]
 lowerarm_r [{shapeType=Capsule, length=14.7535, radius=13.781, center=<-8.9852 0 -1.9278>, rotation=<rot 80.4381 -0.0001 179.9999>}]
 pelvis [
  {shapeType=Capsule, length=19.3878, radius=21.7168, center=<-16.5613 2.2879 3.4499>}
  {shapeType=Capsule, length=19.9892, radius=22.5764, shapeName=pelvis_capsule, center=<-16.8307 2.4127 -2.9886>}
 ]
 spine_02 [{shapeType=Capsule, length=16.6515, radius=18.2411, center=<-5.0595 -1.7096 0>}]
 spine_03 [{shapeType=Capsule, length=5.6604, radius=20.3961, center=<-4.9367 0 0>}]
 spine_04 [{shapeType=Capsule, length=19.5292, radius=22.3505, center=<-1.0437 1.6032 0>}]
 spine_05 [{shapeType=Capsule, length=12.137, radius=25.3947, center=<-1.968 4.9994 0>}]
 thigh_l [{shapeType=Capsule, length=27.9115, radius=12.8304, center=<-28.1731 0 1.2756>, rotation=<rot 90 180 0>}]
 thigh_r [{shapeType=Capsule, length=27.9115, radius=12.8304, center=<28.1731 0 -1.2756>, rotation=<rot 90 0 0>}]
 upperarm_l [{shapeType=Capsule, length=7.3759, radius=21.2086, center=<6.4891 8.4051 4.8876>, rotation=<rot 83.4862 -0.0013 -7.3364>}]
 upperarm_r [{shapeType=Capsule, length=6.546, radius=23.3977, center=<-10.0455 -8.4051 -1.5302>, rotation=<rot 90 -165.7418 0>}]
}
constraints []
