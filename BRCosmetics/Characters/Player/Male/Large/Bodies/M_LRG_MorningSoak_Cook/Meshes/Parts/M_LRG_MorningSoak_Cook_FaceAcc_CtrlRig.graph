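# Control Rig for the face accessory of the large male MorningSoak Cook body, applying pose-driven offsets to the helmet and helmet wire bones.
# Three spherical pose readers each weight a Modify Transforms step, chained in sequence.
# Two read the head against spine_05, moving the helmet and then wire plus helmet.
# The third reads the torso box against the head to shift the wire; the latter two readers have debug drawing enabled.
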ControlRigBlueprint /BRCosmetics/Characters/Player/Male/Large/Bodies/M_LRG_MorningSoak_Cook/Meshes/Parts/M_LRG_MorningSoak_Cook_FaceAcc_CtrlRig
graphs [
 RigVMFunctionLibrary {links [], nodes []}
 RigVMModel {nodes [
   RigUnit_BeginExecution = &RigUnit_BeginExecution()
   n1 = &RigUnit_SphericalPoseReader(
    ActiveRegionScaleFactors=(PositiveWidth=1,NegativeWidth=1,PositiveHeight=1,NegativeHeight=1)
    ActiveRegionSize=0.2
    Debug=(bDrawDebug=false,bDraw2D=False,bDrawLocalAxes=False,DebugScale=25,DebugSegments=20,DebugThickness=0.25)
    DriverAxis=(X=0,Y=1,Z=0)
    DriverItem=(Type=Bone,Name="head")
    FalloffRegionScaleFactors=(PositiveWidth=1,NegativeWidth=1,PositiveHeight=1,NegativeHeight=1)
    FalloffSize=0.6
    FlipHeightScaling=False
    FlipWidthScaling=False
    OptionalParentItem=(Type=Bone,Name="spine_05")
    RotationOffset=(X=90,Y=0,Z=20)
    ExecutePin=@RigUnit_BeginExecution.ExecutePin
   )
   > &RigUnit_ModifyTransforms(
    ItemToModify=((Item=(Type=Bone,Name="dyn_helmet_MSTR"),Transform=(Rotation=(X=0,Y=-0.0175,Z=0,W=0.9998),Translation=(X=1,Y=0,Z=0),Scale3D=(X=1,Y=1,Z=1))))
    WeightMinimum=0
    Weight=@n1.OutputParam
   )
   > n2 = &RigUnit_SphericalPoseReader(
    ActiveRegionScaleFactors=(PositiveWidth=2,NegativeWidth=2,PositiveHeight=2,NegativeHeight=1)
    ActiveRegionSize=0.15
    Debug=(bDrawDebug=true,bDraw2D=False,bDrawLocalAxes=False,DebugScale=25,DebugSegments=20,DebugThickness=0.25)
    DriverAxis=(X=0,Y=1,Z=0)
    DriverItem=(Type=Bone,Name="head")
    FalloffRegionScaleFactors=(PositiveWidth=1,NegativeWidth=1,PositiveHeight=1,NegativeHeight=1)
    FalloffSize=0.6
    FlipHeightScaling=False
    FlipWidthScaling=False
    OptionalParentItem=(Type=Bone,Name="spine_05")
    RotationOffset=(X=150,Y=0,Z=20)
   )
   > &RigUnit_ModifyTransforms(
    ItemToModify="((Item=(Type=Bone,Name=\"dyn_helmet_wire_MSTR\"),Transform=(Rotation=(X=0,Y=0,Z=0,W=1),Translation=(X=2,Y=10,Z=0),Scale3D=(X=1,Y=1,Z=1))),(Item=(Type=Bone,Name=\"dyn_helmet_MSTR\"),Transform=(Rotation=(X=0,Y=0,Z=0,W=1),Translation=(X=1,Y=0,Z=0),Scale3D=(X=1,Y= ...(truncated)"
    WeightMinimum=0
    Weight=@n2.OutputParam
   )
   > n3 = &RigUnit_SphericalPoseReader(
    ActiveRegionScaleFactors=(PositiveWidth=2,NegativeWidth=2,PositiveHeight=2,NegativeHeight=1)
    ActiveRegionSize=0.1
    Debug=(bDrawDebug=true,bDraw2D=False,bDrawLocalAxes=False,DebugScale=25,DebugSegments=20,DebugThickness=0.25)
    DriverAxis=(X=0,Y=1,Z=0)
    DriverItem=(Type=Bone,Name="deform_torso_box_MSTR")
    FalloffRegionScaleFactors=(PositiveWidth=1,NegativeWidth=1,PositiveHeight=1,NegativeHeight=1)
    FalloffSize=0.4
    FlipHeightScaling=False
    FlipWidthScaling=False
    OptionalParentItem=(Type=Bone,Name="head")
    RotationOffset=(X=-40,Y=0,Z=0)
   )
   > &RigUnit_ModifyTransforms(
    ItemToModify=((Item=(Type=Bone,Name="dyn_helmet_wire_MSTR"),Transform=(Rotation=(X=0,Y=0,Z=0,W=-1),Translation=(X=5,Y=5,Z=0),Scale3D=(X=1,Y=1,Z=1))))
    WeightMinimum=0
    Weight=@n3.OutputParam
   )
  ]}
]
hierarchy [
 root {loc=<0 0 0>, rot=<0 0 0 1>}
 pelvis {parent=root, loc=<0 -0.7321 92.4218>, rot=<0.0127 0.707 -0.0127 -0.707>}
 spine_01 {parent=pelvis, loc=<5.6002 -0.5128 0>, rot=<0 0 0.034 -0.9994>}
 spine_02 {parent=spine_01, loc=<8.0756 0.0139 0>, rot=<0 0 0.0238 -0.9997>}
 spine_03 {parent=spine_02, loc=<7.0395 0 0>, rot=<0 0 0.0149 -0.9999>}
 spine_04 {parent=spine_03, loc=<8.1023 0 0>, rot=<0 0 -0.0132 -0.9999>}
 spine_05 {parent=spine_04, loc=<10.0141 0 0>, rot=<0 0 -0.0015 -1>}
 neck_01 {parent=spine_05, loc=<10.815 1.9196 0>, rot=<0 0 -0.1864 -0.9825>}
 neck_02 {parent=neck_01, loc=<4.7347 0.3045 0>, rot=<0 0 0.0363 -0.9993>}
 head {parent=neck_02, loc=<5.7632 0.5552 0>, rot=<0 0 0.115 -0.9934>}
 dyn_helmet_MSTR {parent=head, loc=<11.79 2.005 0.278>, rot=<0.6817 -0.2037 0.1577 -0.6848>}
 dyn_helmet {parent=dyn_helmet_MSTR, loc=<0 0 0>, rot=<0 0 0 1>}
 dyn_antenna_02 {parent=dyn_helmet, loc=<17.8086 -18.1365 -0.9827>, rot=<0.0016 -0.2221 0.1129 -0.9685>}
 dyn_antenna_01 {parent=dyn_helmet, loc=<17.9452 -18.1287 1.2818>, rot=<-0.006 0.1674 -0.0653 -0.9837>}
 dyn_helmet_wire_CONST {parent=head, loc=<5.4535 8.2952 -20.4379>, rot=<0.0029 -0.7071 -0.0029 -0.7071>}
 dyn_helmet_wire_MSTR {parent=dyn_helmet_wire_CONST, loc=<0 0 0>, rot=<0 0 0 1>}
 dyn_helmet_wire {parent=dyn_helmet_wire_MSTR, loc=<0 0 0>, rot=<0 0 0 1>}
 deform_torso_box_MSTR {parent=spine_05, loc=<-2.8479 2.5282 -0.0313>, rot=<0 0 -0.0401 -0.9992>}
 deform_torso_box_front {parent=deform_torso_box_MSTR, loc=<0 19.9578 0>, rot=<-0.5 -0.5 -0.5 -0.5>}
 deform_helmet_wire_bttm_MSTR {parent=spine_05, loc=<9.2225 20.7648 -18.0714>, rot=<0.0284 -0.7065 -0.0284 -0.7065>}
 deform_helmet_wire_bttm {parent=deform_helmet_wire_bttm_MSTR, loc=<0 0 0>, rot=<0 0 0 1>}
]
