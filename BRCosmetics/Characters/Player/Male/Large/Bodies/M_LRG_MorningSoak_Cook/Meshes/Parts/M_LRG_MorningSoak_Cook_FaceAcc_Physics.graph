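# Physics asset for the face accessory part of the large male MorningSoak Cook player body, covering helmet and helmet-wire dynamic bones.
# Capsule bodies sit on the head, helmet, wire and their MSTR/CONST helper bones; the head has a large capsule, the wire two elongated ones, the others tiny ones.
# Constraints tie wire and helmet to their MSTR bones: the wire's swing and twist are locked, while the helmet locks only twist.
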PhysicsAsset /BRCosmetics/Characters/Player/Male/Large/Bodies/M_LRG_MorningSoak_Cook/Meshes/Parts/M_LRG_MorningSoak_Cook_FaceAcc_Physics
bodies {
 dyn_helmet [{shapeType=Capsule, length=1, radius=1, shapeName=dyn_helmet_capsule}]
 dyn_helmet_MSTR [{shapeType=Capsule, length=0.505, radius=0.505, shapeName=dyn_helmet_MSTR_capsule, rotation=<rot 0 -90 0>}]
 dyn_helmet_wire [
  {
   shapeType=Capsule
   length=25.088
   radius=2.6239
   shapeName=dyn_helmet_wire_capsule
   center=<-0.6669 1.8698 -3.1449>
   rotation=<rot -38.192 -90 14.9745>
  }
  {
   shapeType=Capsule
   length=29.368
   radius=2.755
   shapeName=dyn_helmet_wire_capsule2
   center=<0.5349 -2.3664 2.24>
   rotation=<rot -38.1921 -90 1.3487>
  }
 ]
 dyn_helmet_wire_CONST [{shapeType=Capsule, length=0.505, radius=0.505, shapeName=dyn_helmet_wire_CONST_capsule, rotation=<rot 0 -90 0>}]
 dyn_helmet_wire_MSTR [{shapeType=Capsule, length=0.505, radius=0.505, shapeName=dyn_helmet_wire_MSTR_capsule, rotation=<rot 0 -90 0>}]
 head [{shapeType=Capsule, length=2.1733, radius=17.2756, shapeName=head_capsule, center=<5.7222 3.8298 0>}]
}
constraints [
 {
  bone1Name=dyn_helmet_wire
  bone2Name=dyn_helmet_wire_MSTR
  swing1LimitDegrees=5
  swing1Motion=Locked
  swing2LimitDegrees=5
  swing2Motion=Locked
  twistLimitDegrees=2
  twistMotion=Locked
 }
 {bone1Name=dyn_helmet, bone2Name=dyn_helmet_MSTR, swing1LimitDegrees=3, swing2LimitDegrees=5, twistLimitDegrees=1, twistMotion=Locked}
]
