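# Animation Blueprint for the MorningSoak Cook character's face accessory, driving its helmet, wire and antenna motion.
# It copies the parent pose, layers in the body mesh pose from the wire's bottom bone, then runs Control Rig and head-space RigidBody physics.
# A weighted constraint ties the wire to helmet, head and spine before two antenna AnimDynamics nodes.
# Events zero the antenna dynamics alpha during skydiving, grinding and ziplining, restoring it afterward.
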AnimBlueprint /BRCosmetics/Characters/Player/Male/Large/Bodies/M_LRG_MorningSoak_Cook/Meshes/Parts/M_LRG_MorningSoak_Cook_FaceAcc_AnimBP
target_skeleton=&/M_LRG_MorningSoak_Cook_FaceAcc_Skeleton
graphs [
 AnimGraph {nodes [
   n1 = Get BodyMesh() -> (BodyMesh:Skeletal Mesh Component Budgeted Object Reference)
   n2 = Control Rig(Source=Layered blend per bone(
     BasePose=Copy Pose From Mesh() {bUseAttachedParent=true}
     BlendPoses_0=Copy Pose From Mesh(SourceMeshComponent=@n1) {bUseAttachedParent=true}
    ) {layerSetup [{branchFilters [{boneName=deform_helmet_wire_bttm_MSTR}]}]}) {controlRigAssetReference {blueprintRigClass=&/M_LRG_MorningSoak_Cook_FaceAcc_CtrlRig.M_LRG_MorningSoak_Cook_FaceAcc_CtrlRig_C}}
   n3 = RigidBody(ComponentPose=Local To Component(LocalPose=@n2)) {
    bOverrideWorldGravity=true
    overridePhysicsAsset=&/M_LRG_MorningSoak_Cook_FaceAcc_Physics
    simulationSpace=BaseBoneSpace
    baseBoneRef {boneName=head}
    overrideWorldGravity {z=-980}
    simSpaceSettings {
     dampingAlpha=0
     maxAngularAcceleration=30
     maxAngularVelocity=6
     maxLinearAcceleration=3000
     maxLinearVelocity=600
     worldAlpha=1
     externalLinearDragV=<0.1 0.1 0.1>
    }
   }
   n4 = "Constraint - dyn_helmet_wire_CONST (PPP)"!(ComponentPose=@n3, Alpha=1, ConstraintWeights_0=0.3, ConstraintWeights_1:float=0.3, ConstraintWeights_2:float=0.4) -> (Pose) {
    boneToModify {boneName=dyn_helmet_wire_CONST}
    constraintSetup [{targetBone {boneName=dyn_helmet}}, {targetBone {boneName=head}}, {targetBone {boneName=spine_05}}]
    constraintWeights [1,1,1]
   }
   n5 = Get AnimDynamicsAlpha() -> (AnimDynamicsAlpha:double)
   n6 = Anim Dynamics - Bone: dyn_antenna_01(ComponentPose=@n4, Alpha=@n5, ExternalForce="0,0,0") {
    angularDampingOverride=0.999
    angularSpringConstant=20000
    bAngularSpring=true
    bOverrideAngularDamping=true
    bUseGravityOverride=true
    simulationSpace=BoneRelative
    boundBone {boneName=dyn_antenna_01}
    componentAppliedLinearAccClamp=<1500 1500 1500>
    gravityOverride {z=-980}
    physicsBodyDefinitions [{
      boundBone {boneName=dyn_antenna_01}
      boxExtents=<40 40 40>
      constraintSetup {angularLimitsMax=<4 4 4>, angularLimitsMin=<-4 -4 -4>, angularTarget=<1 0 0>}
      localJointOffset=<3 0 0>
     }]
    relativeSpaceBone {boneName=pelvis}
    simSpaceSettings {maxAngularAcceleration=40, maxAngularVelocity=8, simSpaceAngularAlpha=0.4}
   }
   n7 = Anim Dynamics - Bone: dyn_antenna_02(ComponentPose=@n6, Alpha=@n5, ExternalForce="0,0,0") {
    angularDampingOverride=0.999
    angularSpringConstant=20000
    bAngularSpring=true
    bOverrideAngularDamping=true
    bUseGravityOverride=true
    simulationSpace=BoneRelative
    boundBone {boneName=dyn_antenna_02}
    componentAppliedLinearAccClamp=<1500 1500 1500>
    gravityOverride {z=-980}
    physicsBodyDefinitions [{
      boundBone {boneName=dyn_antenna_02}
      boxExtents=<35 35 35>
      constraintSetup {angularLimitsMax=<4 4 4>, angularLimitsMin=<-4 -4 -4>, angularTarget=<1 0 0>}
      localJointOffset=<-1.5 0 0>
     }]
    relativeSpaceBone {boneName=pelvis}
    simSpaceSettings {maxAngularAcceleration=40, maxAngularVelocity=8, simSpaceAngularAlpha=0.4}
   }
   Output Pose(Result=Component To Local(ComponentPose=@n7)) {layerGroup=None}
  ]}
 EventGraph {nodes [
   Event OnBeginSkydiving()
   > Set AnimDynamicsAlpha(AnimDynamicsAlpha:double=0) -> (Output_Get:double)
   Event OnEndSkydiving()
   > Set AnimDynamicsAlpha(AnimDynamicsAlpha=1)
   Event OnEnterGrindRail() {is=title K2Node_Event}
   > Set AnimDynamicsAlpha(AnimDynamicsAlpha=0)
   n1 = Event BlueprintProcessAnimInputEvent()
   > n2 = Branch(Condition=IsNamed(self=@n1.AnimInputEvent, InEventName=aie.Player.Zipline.Entered))
   > Set AnimDynamicsAlpha(AnimDynamicsAlpha=0)
   Branch(execute=@n2.else, Condition=IsNamed(self=@n1.AnimInputEvent, InEventName=aie.Player.Zipline.Exited))
   > Set AnimDynamicsAlpha(AnimDynamicsAlpha=1)
   Event OnExitGrindRail() {is=title K2Node_Event}
   > Set AnimDynamicsAlpha(AnimDynamicsAlpha=1)
   n3 = Event OnCharacterCustomizationLoaded()
   > Set BodyMesh(BodyMesh:Skeletal Mesh Component Budgeted Object Reference=Get Array Item with EFort Custom Part Type(Array=@n3.CustomMeshParts)) -> (Output_Get:Skeletal Mesh Component Budgeted Object Reference)
   Event BlueprintUpdateAnimation()
  ]}
]
variables [
 AnimDynamicsAlpha {type=number, default=1}
 BodyMesh {type=SkeletalMeshComponentBudgeted}
]
